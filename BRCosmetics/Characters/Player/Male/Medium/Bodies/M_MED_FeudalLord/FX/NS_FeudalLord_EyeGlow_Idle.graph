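# Idle eye-glow Niagara system for the male Feudal Lord body cosmetic, drawn as small brightened, semi-transparent camera-offset sprites.
# One emitter bursts two particles, partitions them by execution index, and pins each to its own skeletal mesh bone or socket every update.
# The partition also sets VisibilityTag for the two sprite renderers, and User.bisHeadHidden zeroes alpha.
# A per-eye alpha fade based on sampled bone orientation dot products is present but disabled.
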NiagaraSystem /BRCosmetics/Characters/Player/Male/Medium/Bodies/M_MED_FeudalLord/FX/NS_FeudalLord_EyeGlow_Idle
dependencies {
 usedDataInterfaces [
  &NiagaraDataInterfaceArrayInt32
  &NiagaraDataInterfaceArrayFloat3
  &NiagaraDataInterfaceSkeletalMesh
 ]
 usedDynamicInputs []
 usedModules [
  &self:SystemSpawnScript.SystemScriptSource.SystemScriptGraph.NiagaraNodeAssignment_1.SetVariables_#1
  &SystemState
 ]
}
emitters {EyeGlows {
  emitterUpdateScript [
   &EmitterState(
    Inactive Response=::NewEnumerator1
    Life Cycle Mode=::NewEnumerator1
    Loop Behavior=::NewEnumerator1
    Loop Duration Mode=::NewEnumerator1
    UseLoopDelay=false
   )
   &SpawnBurst_Instantaneous(Spawn Count=2, Use Spawn Probability=false, UseLoopCountLimit=true)
  ]
  particleSpawnScript [
   &InitializeParticle(
    Color=<rgba 0.266 0.266 0.266 1>
    Sprite Size Mode=::NewEnumerator3
    Uniform Sprite Size=6
    UsePositionOffset=false
    Write Material Random=true
   )
   &PartitionParticles(ID to Partition:NiagaraInt32=&ReturnExecIndex())
   &CameraOffset(Camera Offset Amount=4)
  ]
  particleUpdateScript [
   &ParticleState(Kill Particles When Lifetime Has Elapsed=false, Let Infinitely Lived Particles Die When Emitter Deactivates=true)
   &ScaleColor(
    Scale Alpha=&MakeCustomFloatFromBool(Bool=$User.bisHeadHidden, True Float=0, False Float=0.35)
    Scale RGB=&MakeVectorFromLinearColorRGB(Color=<rgba 3 3 3 1>)
    ScaleA=true
    ScaleRGB=true
   )
   /SetVariables(Particles.VisibilityTag=$Particles.PartitionParticles.Partition) {isSetParameters=true}
   &SkeletalMeshLocation(
    Bone / Socket Index=&SelectIntFromArray(Direct Array Index=$Particles.PartitionParticles.Partition, Int Selection Array=DataInterface() {propertyValues {IntData [0,1]}})
    MaskSpawnGroup=false
    Mesh Sampling Type=::NewEnumerator5
    Sampled Position Offset=&SelectVectorFromArray(
     Direct Array Index=$Particles.PartitionParticles.Partition
     Vector Selection Array=DataInterface() {propertyValues {FloatData [<0 0 0>, <0 0 0>]}}
    )
    Skeletal Mesh=$System.Skeletal Mesh
    Socket Sampling Mode=::NewEnumerator5
   )
   &SampleSkeletalMesh(
    AdditionalYawPitchRoll=false
    Bone / Socket Index=1
    Bone Sampling Mode:ENiagaraBoneSamplingMode=::NewEnumerator4
    Position Offset Coordinate Space:ENiagaraCoordinateSpace=::Local
    Sampled Orientation Coordinate Space:ENiagaraCoordinateSpace=::Local
    Skeletal Mesh:NiagaraDataInterfaceSkeletalMesh=$System.Skeletal Mesh
   ) {enabled=false}
   &SampleSkeletalMesh(
    AdditionalYawPitchRoll=false
    Bone Sampling Mode:ENiagaraBoneSamplingMode=::NewEnumerator4
    Position Offset Coordinate Space:ENiagaraCoordinateSpace=::Local
    Sampled Orientation Coordinate Space:ENiagaraCoordinateSpace=::Local
    Skeletal Mesh:NiagaraDataInterfaceSkeletalMesh=$System.Skeletal Mesh
   ) {enabled=false, module=SampleSkeletalMesh001}
   /SetVariables(Particles.ColorLeftEye:LinearColor=&MultiplyLinearColor(
     Linear Color=$Particles.Color
     Scale Factor=&MakeFloat4FromVectorAndFloat(
      "Vector (RGB)"=<1 1 1>
      "Float (Alpha)"=&RemapRange(
       Input Value To Remap=&DotProduct(
        Vector A=&ReturnMeshOrientationAxis(Mesh Orientation Axis=::YAxis, Mesh Orientation Quat=$Particles.SampleSkeletalMesh.SampledRotation)
        Vector B=&Multiply_Vector(
         A=<1 1 -1>
         B=&ReturnMeshOrientationAxis(Mesh Orientation Axis=::ZAxis, Mesh Orientation Quat=$Particles.SampleSkeletalMesh001.SampledRotation)
        )
       )
       Input Min=0
       Input Max=1
       Output Min=0
       Output Max=0.7
      )
     )
    )) {enabled=false, isSetParameters=true}
   &SampleSkeletalMesh(
    AdditionalYawPitchRoll=false
    Bone / Socket Index=2
    Bone Sampling Mode:ENiagaraBoneSamplingMode=::NewEnumerator4
    Position Offset Coordinate Space:ENiagaraCoordinateSpace=::Local
    Sampled Orientation Coordinate Space:ENiagaraCoordinateSpace=::Local
    Skeletal Mesh:NiagaraDataInterfaceSkeletalMesh=$System.Skeletal Mesh
   ) {enabled=false, module=SampleSkeletalMesh002}
   &SampleSkeletalMesh(
    AdditionalYawPitchRoll=false
    Bone / Socket Index=3
    Bone Sampling Mode:ENiagaraBoneSamplingMode=::NewEnumerator4
    Position Offset Coordinate Space:ENiagaraCoordinateSpace=::Local
    Sampled Orientation Coordinate Space:ENiagaraCoordinateSpace=::Local
    Skeletal Mesh:NiagaraDataInterfaceSkeletalMesh=$System.Skeletal Mesh
   ) {enabled=false, module=SampleSkeletalMesh003}
   /SetVariables(Particles.ColorRightEye:LinearColor=&MultiplyLinearColor(
     Linear Color=$Particles.Color
     Scale Factor=&MakeFloat4FromVectorAndFloat(
      "Vector (RGB)"=<1 1 1>
      "Float (Alpha)"=&NormalizeFloat(
       Value=&RemapRange(
        Input Value To Remap=&DotProduct(
         Vector A=&ReturnMeshOrientationAxis(Mesh Orientation Axis=::YAxis, Mesh Orientation Quat=$Particles.SampleSkeletalMesh002.SampledRotation)
         Vector B=&ReturnMeshOrientationAxis(Mesh Orientation Axis=::ZAxis, Mesh Orientation Quat=$Particles.SampleSkeletalMesh003.SampledRotation)
        )
        Input Min=0
        Input Max=1
        Output Min=0.7
        Output Max=0
       )
       Min=0
       Max=1
      )
     )
    )) {enabled=false, isSetParameters=true}
   /SetVariables(Particles.Debug:NiagaraFloat=&MakeFloatFromLinearColor(LinearColor=$Particles.ColorLeftEye, Channel:ENiagaraLinearColor_Channels=::NewEnumerator3)) {enabled=false, isSetParameters=true}
  ]
  rendererClasses [&NiagaraSpriteRendererProperties]
  renderers [
   {rendererClass=&NiagaraSpriteRendererProperties}
   {rendererClass=&NiagaraSpriteRendererProperties, rendererIndex=1}
  ]
 }}
summary {
 systemName=NS_FeudalLord_EyeGlow_Idle
 emitters [{emitterName=EyeGlows, rendererClasses [&NiagaraSpriteRendererProperties]}]
 userVariables [
  User.bIsFrontEnd {type=NiagaraBool}
  User.bIsFrontEndPreview {type=NiagaraBool}
  User.bisHeadHidden {type=NiagaraBool}
  User.bKillReaction {type=NiagaraBool}
  User.IsSkydiving {type=NiagaraBool}
  User.showFX {type=NiagaraFloat, defaultValue=1, description=""}
 ]
}
