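# Control Rig for the FeudalLord male body that procedurally corrects deformation and armor bones.
# A sequence runs the shared DeformRig function, derives crouch height from pelvis Z, then offsets armguards and kneepads using deformRig attributes.
# Spherical pose readers on thighs and left upper arm drive skirt and shoulderpad offsets, adjusted by twist and movement-state variables (sliding, sprint, crouch, airborne).
# The control-to-bone limb chain has no execution input.
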ControlRigBlueprint /BRCosmetics/Characters/Player/Male/Medium/Bodies/M_MED_FeudalLord/Meshes/M_MED_FeudalLord_CtrlRig
graphs [
 Multiply_1_ContainedGraph {nodes [
   n1 = RigVMFunctionEntryNode()
   RigVMFunctionReturnNode(Result=&RigVMFunction_MathDoubleMul(B=@n1.C, A=&RigVMFunction_MathDoubleMul(A=@n1.A, B=@n1.B).Result).Result)
  ]}
 New Function_ContainedGraph {nodes [
   n1 = RigVMFunctionEntryNode()
   RigVMFunction_MathBoolToInteger = &RigVMFunction_MathBoolToInteger(Value=@n1.onlyTwist)
   RigVMFunction_MathQuaternionSwingTwist = &RigVMFunction_MathQuaternionSwingTwist(
    TwistAxis=@n1.twistAxis
    Input=&RigUnit_GetRelativeTransformForItem(
     Child=(Type=Bone,Name="None")
     Parent=(Type=Bone,Name="None")
     bChildInitial=False
     bParentInitial=False
     Child.Name=@n1.child
     Parent.Name=@n1.parent
     Parent.Type=@n1.parentType
    ).RelativeTransform.Rotation
   )
   RigVMFunction_MathQuaternionToEuler = &RigVMFunction_MathQuaternionToEuler(
    Value=RigVMDispatchNode(
     Values=((),())
     Index=@RigVMFunction_MathBoolToInteger.Result
     Values.0=@RigVMFunction_MathQuaternionSwingTwist.Swing
     Values.1=@RigVMFunction_MathQuaternionSwingTwist.Twist
    ) {template="DISPATCH_RigVMDispatch_SelectInt32(in Index,in Values,out Result)"}.Result
    RotationOrder=@n1.Rotation Order
   )
   > DISPATCH_RigVMDispatch_Print = RigVMDispatchNode(ScreenColor=(R=1,G=1,B=1,A=1), ScreenDuration=1.5, Enabled=@n1.debugRotation, Value=@RigVMFunction_MathQuaternionToEuler.Result) {template="DISPATCH_RigVMDispatch_Print(in Prefix,in Value,in Enabled,in ScreenDuration,in ScreenColor)"}
   Clamp = &RigVMFunction_MathIntClamp(Value=@n1.primAxis)
   DISPATCH_RigVMDispatch_SelectInt32 = RigVMDispatchNode(
    Values=(0,0,0)
    Index=@Clamp.Result
    Values.0=@RigVMFunction_MathQuaternionToEuler.Result.X
    Values.1=@RigVMFunction_MathQuaternionToEuler.Result.Y
    Values.2=@RigVMFunction_MathQuaternionToEuler.Result.Z
   ) {template="DISPATCH_RigVMDispatch_SelectInt32(in Index,in Values,out Result)"}
   Remap = &RigVMFunction_MathDoubleRemap(
    TargetMaximum=1
    TargetMinimum=0
    bClamp=False
    Value=@DISPATCH_RigVMDispatch_SelectInt32.Result
    SourceMaximum=@n1.max
    SourceMinimum=@n1.min
   )
   Multiply = &RigVMFunction_MathDoubleMul(A=&RigVMFunction_MathDoubleClamp(Value=@Remap.Result).Result, B=@n1.premult)
   > RigVMFunctionReturnNode(outputWeight=@Multiply.Result)
  ]}
 RigVMFunctionLibrary {nodes [DeltaRotation = RigVMCollapseNode(ExecuteContext=(), Rotation Order=ZYX, premult=1, twistAxis=(X=1,Y=0,Z=0))]}
 RigVMFunction_Sequence_1_ContainedGraph {nodes [
   n1 = RigVMFunctionEntryNode()
   RigVMFunction_Sequence = &RigVMFunction_Sequence(ExecuteContext=@n1.ExecuteContext)
   n2 = &RigVMFunction_Sequence(ExecuteContext=@RigVMFunction_Sequence.B)
   n3 = &RigVMFunction_Sequence(ExecuteContext=@n2.B)
   n4 = &RigVMFunction_Sequence(ExecuteContext=@n3.B)
   n5 = &RigVMFunction_Sequence(ExecuteContext=@n4.B)
   n6 = &RigVMFunction_Sequence(ExecuteContext=@n5.B)
   RigVMFunctionReturnNode(A=@RigVMFunction_Sequence.A, B=@n2.A, C=@n3.A, D=@n4.A, E=@n5.A, F=@n6.A, G=@n6.B)
  ]}
 RigVMModel {nodes [
   n1 = RigVMRerouteNode(Value=())
   &RigUnit_SetTransform(
    Item=(Type=Bone,Name="clavicle_l")
    Weight=1
    bInitial=False
    Value=&RigUnit_GetTransform(Item=(Type=Control,Name="clavicle_l_ctrl"), bInitial=False).Transform
    ExecutePin=@n1.Value
   )
   > &RigUnit_SetTransform(
    Item=(Type=Bone,Name="upperarm_l")
    Weight=1
    bInitial=False
    Value=&RigUnit_GetTransform(Item=(Type=Control,Name="upperarm_l_ctrl"), bInitial=False).Transform
   )
   > &RigUnit_SetTransform(
    Item=(Type=Bone,Name="lowerarm_l")
    Weight=1
    bInitial=False
    Value=&RigUnit_GetTransform(Item=(Type=Control,Name="lowerarm_l_ctrl"), bInitial=False).Transform
   )
   > &RigUnit_SetTransform(
    Item=(Type=Bone,Name="hand_l")
    Weight=1
    bInitial=False
    Value=&RigUnit_GetTransform(Item=(Type=Control,Name="hand_l_ctrl"), bInitial=False).Transform
   )
   > &RigUnit_SetTransform(
    Item=(Type=Bone,Name="upperarm_r")
    Weight=1
    bInitial=False
    Value=&RigUnit_GetTransform(Item=(Type=Control,Name="upperarm_r_ctrl"), bInitial=False).Transform
   )
   > &RigUnit_SetTransform(
    Item=(Type=Bone,Name="lowerarm_r")
    Weight=1
    bInitial=False
    Value=&RigUnit_GetTransform(Item=(Type=Control,Name="lowerarm_r_ctrl"), bInitial=False).Transform
   )
   > &RigUnit_SetTransform(
    Item=(Type=Bone,Name="hand_r")
    Weight=1
    bInitial=False
    Value=&RigUnit_GetTransform(Item=(Type=Control,Name="hand_r_ctrl"), bInitial=False).Transform
   )
   > RigUnit_SetTransform = &RigUnit_SetTransform(
    Item=(Type=Bone,Name="thigh_l")
    Weight=1
    bInitial=False
    Value=&RigUnit_GetTransform(Item=(Type=Control,Name="thigh_l_ctrl"), bInitial=False).Transform
   )
   > &RigUnit_SetTransform(
    Item=(Type=Bone,Name="calf_l")
    Weight=1
    bInitial=False
    Value=&RigUnit_GetTransform(Item=(Type=Control,Name="calf_l_ctrl"), bInitial=False).Transform
   )
   RigUnit_GetTransform = &RigUnit_GetTransform(Item=(Type=Control,Name="thigh_r_ctrl"), bInitial=False)
   > &RigUnit_SetTransform(Item=(Type=Bone,Name="thigh_r"), Weight=1, bInitial=False, Value=@RigUnit_GetTransform.Transform)
   > &RigUnit_SetTransform(
    Item=(Type=Bone,Name="calf_r")
    Weight=1
    bInitial=False
    Value=&RigUnit_GetTransform(Item=(Type=Control,Name="calf_r_ctrl"), bInitial=False).Transform
   )
   RigUnit_BeginExecution = &RigUnit_BeginExecution()
   RigVMFunction_Sequence = RigVMAggregateNode(ExecuteContext=@RigUnit_BeginExecution.ExecutePin)
   FN_DeformRig_FNC = RigVMFunctionReferenceNode(
    DeformRig_ Curve="(deformRig_l_upperarm_up_alpha_72_#1=False,deformRig_r_upperarm_up_alpha_73_#2=False,deformRig_l_upperarm_fr_alpha_74_#3=False,deformRig_r_upperarm_fr_alpha_75_CB51B8 ...(truncated)"
    DeformRig_Bool="(upperarm_1_#4=True,elbow_3_#5=True,wrist_7_#6=True,groin_8_#7=True,glute_17_#8=True,knee_9_40C408324CA261 ...(truncated)"
    DeformRig_Debug=(upperarm_debug_18_#4=False,wrist_debug_20_#6=False,lat_debug_26_#9=False,groinGlute_debug_27_#7=False)
    DeformRig_Value="(upperarm_up_18_#10=(Rotation=(X=0.0,Y=-0.0,Z=0.0,W=1.0),Translation=(X=0.0,Y=0.0,Z=8.0),Scale3D=(X=0.70,Y=1.0,Z=1.0)),upperarm_fwd_fr_19_#11=(Rotation=(X=0.0,Y=-0.0,Z=0.0,W=1.0),Translation=(X=0.0, ...(truncated)"
    ExecuteContext=@RigVMFunction_Sequence.B
   )
   RigVMFunction_MathBoolMake = /Script/RigVM.RigVMFunction_MathBoolMake(Value=false) {is=struct RigVMUnitNode}
   n2 = &RigUnit_SphericalPoseReader(
    ActiveRegionScaleFactors=(PositiveWidth=1,NegativeWidth=1,PositiveHeight=1,NegativeHeight=1)
    ActiveRegionSize=0.05
    Debug=(bDrawDebug=false,bDraw2D=False,bDrawLocalAxes=False,DebugScale=25,DebugSegments=20,DebugThickness=0.25)
    DriverAxis=(X=1,Y=0,Z=0)
    DriverItem=(Type=Bone,Name="thigh_r")
    FalloffRegionScaleFactors=(PositiveWidth=1,NegativeWidth=1,PositiveHeight=1,NegativeHeight=1)
    FalloffSize=0.55
    FlipHeightScaling=true
    FlipWidthScaling=true
    OptionalParentItem=(Type=Bone,Name="None")
    RotationOffset=(X=-90,Y=0,Z=-20)
    Debug.bDrawDebug=@RigVMFunction_MathBoolMake.Value
    ExecutePin=@RigVMFunction_Sequence.G
   )
   > &RigUnit_ModifyTransforms(
    ItemToModify="((Item=(Type=Bone,Name=\"auto_skirt_front_rt\"),Transform=(Rotation=(X=0.1234,Y=-0.5782,Z=-0.0548,W=0.8046),Translation=(X=-14.5,Y=0,Z=-1.2),Scale3D=(X=1,Y=1,Z=1))),(Item=(Type=Bone,Name=\"auto_skirt_mid_rt\"),Transform=(Rotation=(X=-0.099,Y=-0.2391,Z=-0.3696, ...(truncated)"
    WeightMinimum=0
    Weight=@n2.OutputParam
   )
   > n3 = &RigUnit_SphericalPoseReader(
    ActiveRegionScaleFactors=(PositiveWidth=1,NegativeWidth=1,PositiveHeight=1,NegativeHeight=1)
    ActiveRegionSize=0.05
    Debug=(bDrawDebug=false,bDraw2D=False,bDrawLocalAxes=False,DebugScale=25,DebugSegments=20,DebugThickness=0.25)
    DriverAxis=(X=1,Y=0,Z=0)
    DriverItem=(Type=Bone,Name="thigh_r")
    FalloffRegionScaleFactors=(PositiveWidth=1,NegativeWidth=1,PositiveHeight=1,NegativeHeight=1)
    FalloffSize=0.55
    FlipHeightScaling=true
    FlipWidthScaling=true
    OptionalParentItem=(Type=Bone,Name="None")
    RotationOffset=(X=-40,Y=20,Z=0)
    Debug.bDrawDebug=@RigVMFunction_MathBoolMake.Value
   )
   > &RigUnit_ModifyTransforms(
    ItemToModify="((Item=(Type=Bone,Name=\"auto_skirt_front_rt\"),Transform=(Rotation=(X=0,Y=-0.1736,Z=0,W=0.9848),Translation=(X=0,Y=0,Z=0),Scale3D=(X=1,Y=1,Z=1))),(Item=(Type=Bone,Name=\"auto_skirt_mid_rt\"),Transform=(Rotation=(X=0.0038,Y=-0.0435,Z=-0.0871,W=0.9952),Translat ...(truncated)"
    WeightMinimum=0
    Weight=@n3.OutputParam
   )
   > n4 = &RigUnit_SphericalPoseReader(
    ActiveRegionScaleFactors=(PositiveWidth=1,NegativeWidth=1,PositiveHeight=1,NegativeHeight=1)
    ActiveRegionSize=0.05
    Debug=(bDrawDebug=false,bDraw2D=False,bDrawLocalAxes=False,DebugScale=25,DebugSegments=20,DebugThickness=0.25)
    DriverAxis=(X=1,Y=0,Z=0)
    DriverItem=(Type=Bone,Name="thigh_r")
    FalloffRegionScaleFactors=(PositiveWidth=1,NegativeWidth=1,PositiveHeight=1,NegativeHeight=1)
    FalloffSize=0.55
    FlipHeightScaling=true
    FlipWidthScaling=true
    OptionalParentItem=(Type=Bone,Name="None")
    RotationOffset=(X=0,Y=20,Z=0)
    Debug.bDrawDebug=@RigVMFunction_MathBoolMake.Value
   )
   > &RigUnit_ModifyTransforms(
    ItemToModify="((Item=(Type=Bone,Name=\"auto_skirt_front_rt\"),Transform=(Rotation=(X=0.0626,Y=-0.4309,Z=0.172,W=0.8837),Translation=(X=0,Y=0,Z=0),Scale3D=(X=1,Y=1,Z=1))),(Item=(Type=Bone,Name=\"auto_skirt_mid_rt\"),Transform=(Rotation=(X=-0.4981,Y=-0.4564,Z=-0.4981,W=0.5436 ...(truncated)"
    WeightMinimum=0
    Weight=@n4.OutputParam
   )
   > n5 = &RigUnit_SphericalPoseReader(
    ActiveRegionScaleFactors=(PositiveWidth=1,NegativeWidth=1,PositiveHeight=1,NegativeHeight=1)
    ActiveRegionSize=0.05
    Debug=(bDrawDebug=false,bDraw2D=False,bDrawLocalAxes=False,DebugScale=25,DebugSegments=20,DebugThickness=0.25)
    DriverAxis=(X=1,Y=0,Z=0)
    DriverItem=(Type=Bone,Name="thigh_r")
    FalloffRegionScaleFactors=(PositiveWidth=1,NegativeWidth=1,PositiveHeight=1,NegativeHeight=1)
    FalloffSize=0.55
    FlipHeightScaling=true
    FlipWidthScaling=true
    OptionalParentItem=(Type=Bone,Name="None")
    RotationOffset=(X=40,Y=20,Z=0)
    Debug.bDrawDebug=@RigVMFunction_MathBoolMake.Value
   )
   > &RigUnit_ModifyTransforms(
    ItemToModify="((Item=(Type=Bone,Name=\"auto_skirt_front_rt\"),Transform=(Rotation=(X=0,Y=0,Z=0,W=1),Translation=(X=0,Y=0,Z=0),Scale3D=(X=1,Y=1,Z=1))),(Item=(Type=Bone,Name=\"auto_skirt_mid_rt\"),Transform=(Rotation=(X=0,Y=0,Z=0.0872,W=0.9962),Translation=(X=0,Y=0,Z=0),Scale ...(truncated)"
    WeightMinimum=0
    Weight=@n5.OutputParam
   )
   > n6 = &RigUnit_SphericalPoseReader(
    ActiveRegionScaleFactors=(PositiveWidth=1,NegativeWidth=1,PositiveHeight=1,NegativeHeight=1)
    ActiveRegionSize=0.05
    Debug=(bDrawDebug=false,bDraw2D=False,bDrawLocalAxes=False,DebugScale=25,DebugSegments=20,DebugThickness=0.25)
    DriverAxis=(X=1,Y=0,Z=0)
    DriverItem=(Type=Bone,Name="thigh_r")
    FalloffRegionScaleFactors=(PositiveWidth=1,NegativeWidth=1,PositiveHeight=1,NegativeHeight=1)
    FalloffSize=0.55
    FlipHeightScaling=true
    FlipWidthScaling=true
    OptionalParentItem=(Type=Bone,Name="None")
    RotationOffset=(X=90,Y=0,Z=20)
    Debug.bDrawDebug=@RigVMFunction_MathBoolMake.Value
   )
   > &RigUnit_ModifyTransforms(
    ItemToModify="((Item=(Type=Bone,Name=\"auto_skirt_front_rt\"),Transform=(Rotation=(X=0,Y=0,Z=0,W=1),Translation=(X=0,Y=0,Z=0),Scale3D=(X=1,Y=1,Z=1))),(Item=(Type=Bone,Name=\"auto_skirt_mid_rt\"),Transform=(Rotation=(X=0,Y=0,Z=0,W=1),Translation=(X=0,Y=0,Z=0),Scale3D=(X=1,Y= ...(truncated)"
    WeightMinimum=0
    Weight=@n6.OutputParam
   )
   n7 = &RigVMFunction_MathBoolMake(Value=false)
   > n8 = &RigUnit_SphericalPoseReader(
    ActiveRegionScaleFactors=(PositiveWidth=1,NegativeWidth=1,PositiveHeight=1,NegativeHeight=1)
    ActiveRegionSize=0.05
    Debug=(bDrawDebug=false,bDraw2D=False,bDrawLocalAxes=False,DebugScale=25,DebugSegments=20,DebugThickness=0.25)
    DriverAxis=(X=-1,Y=0,Z=0)
    DriverItem=(Type=Bone,Name="thigh_l")
    FalloffRegionScaleFactors=(PositiveWidth=1,NegativeWidth=1,PositiveHeight=1,NegativeHeight=1)
    FalloffSize=0.55
    FlipHeightScaling=False
    FlipWidthScaling=False
    OptionalParentItem=(Type=Bone,Name="None")
    RotationOffset=(X=90,Y=0,Z=-20)
    Debug.bDrawDebug=@n7.Value
   )
   > &RigUnit_ModifyTransforms(
    ItemToModify="((Item=(Type=Bone,Name=\"auto_skirt_front_lf\"),Transform=(Rotation=(X=0.1584,Y=-0.35,Z=0.312,W=0.8689),Translation=(X=-14.5002,Y=-0.0001,Z=-1.2001),Scale3D=(X=1,Y=1,Z=1))),(Item=(Type=Bone,Name=\"auto_skirt_mid_lf\"),Transform=(Rotation=(X=-0.2087,Y=0.0615,Z= ...(truncated)"
    WeightMinimum=0
    Weight=@n8.OutputParam
   )
   > n9 = &RigUnit_SphericalPoseReader(
    ActiveRegionScaleFactors=(PositiveWidth=1,NegativeWidth=1,PositiveHeight=1,NegativeHeight=1)
    ActiveRegionSize=0.05
    Debug=(bDrawDebug=false,bDraw2D=False,bDrawLocalAxes=False,DebugScale=25,DebugSegments=20,DebugThickness=0.25)
    DriverAxis=(X=-1,Y=0,Z=0)
    DriverItem=(Type=Bone,Name="thigh_l")
    FalloffRegionScaleFactors=(PositiveWidth=1,NegativeWidth=1,PositiveHeight=1,NegativeHeight=1)
    FalloffSize=0.55
    FlipHeightScaling=False
    FlipWidthScaling=False
    OptionalParentItem=(Type=Bone,Name="None")
    RotationOffset=(X=140,Y=30,Z=0)
    Debug.bDrawDebug=@n7.Value
   )
   > &RigUnit_ModifyTransforms(
    ItemToModify="((Item=(Type=Bone,Name=\"auto_skirt_front_lf\"),Transform=(Rotation=(X=0,Y=0.5,Z=0,W=-0.866),Translation=(X=0,Y=0,Z=0),Scale3D=(X=1,Y=1,Z=1))),(Item=(Type=Bone,Name=\"auto_skirt_mid_lf\"),Transform=(Rotation=(X=0.0226,Y=0.25,Z=-0.2578,W=-0.933),Translation=(X= ...(truncated)"
    WeightMinimum=0
    Weight=@n9.OutputParam
   )
   > n10 = &RigUnit_SphericalPoseReader(
    ActiveRegionScaleFactors=(PositiveWidth=1,NegativeWidth=1,PositiveHeight=1,NegativeHeight=1)
    ActiveRegionSize=0.05
    Debug=(bDrawDebug=false,bDraw2D=False,bDrawLocalAxes=False,DebugScale=25,DebugSegments=20,DebugThickness=0.25)
    DriverAxis=(X=-1,Y=0,Z=0)
    DriverItem=(Type=Bone,Name="thigh_l")
    FalloffRegionScaleFactors=(PositiveWidth=1,NegativeWidth=1,PositiveHeight=1,NegativeHeight=1)
    FalloffSize=0.55
    FlipHeightScaling=False
    FlipWidthScaling=False
    OptionalParentItem=(Type=Bone,Name="None")
    RotationOffset=(X=180,Y=20,Z=0)
    Debug.bDrawDebug=@n7.Value
   )
   > &RigUnit_ModifyTransforms(
    ItemToModify="((Item=(Type=Bone,Name=\"auto_skirt_front_lf\"),Transform=(Rotation=(X=0,Y=0.3007,Z=0,W=-0.9537),Translation=(X=0,Y=0,Z=0),Scale3D=(X=1,Y=1,Z=1))),(Item=(Type=Bone,Name=\"auto_skirt_mid_lf\"),Transform=(Rotation=(X=0.1156,Y=0.548,Z=0.1297,W=-0.8182),Translatio ...(truncated)"
    WeightMinimum=0
    Weight=@n10.OutputParam
   )
   > n11 = &RigUnit_SphericalPoseReader(
    ActiveRegionScaleFactors=(PositiveWidth=1,NegativeWidth=1,PositiveHeight=1,NegativeHeight=1)
    ActiveRegionSize=0.05
    Debug=(bDrawDebug=false,bDraw2D=False,bDrawLocalAxes=False,DebugScale=25,DebugSegments=20,DebugThickness=0.25)
    DriverAxis=(X=-1,Y=0,Z=0)
    DriverItem=(Type=Bone,Name="thigh_l")
    FalloffRegionScaleFactors=(PositiveWidth=1,NegativeWidth=1,PositiveHeight=1,NegativeHeight=1)
    FalloffSize=0.55
    FlipHeightScaling=False
    FlipWidthScaling=False
    OptionalParentItem=(Type=Bone,Name="None")
    RotationOffset=(X=-140,Y=30,Z=0)
    Debug.bDrawDebug=@n7.Value
   )
   > &RigUnit_ModifyTransforms(
    ItemToModify="((Item=(Type=Bone,Name=\"auto_skirt_front_lf\"),Transform=(Rotation=(X=0,Y=0,Z=0.1736,W=-0.9848),Translation=(X=0,Y=0,Z=0),Scale3D=(X=1,Y=1,Z=1))),(Item=(Type=Bone,Name=\"auto_skirt_mid_lf\"),Transform=(Rotation=(X=0,Y=0.0851,Z=0.2164,W=-0.9726),Translation=(X ...(truncated)"
    WeightMinimum=0
    Weight=@n11.OutputParam
   )
   > n12 = &RigUnit_SphericalPoseReader(
    ActiveRegionScaleFactors=(PositiveWidth=1,NegativeWidth=1,PositiveHeight=1,NegativeHeight=1)
    ActiveRegionSize=0.05
    Debug=(bDrawDebug=false,bDraw2D=False,bDrawLocalAxes=False,DebugScale=25,DebugSegments=20,DebugThickness=0.25)
    DriverAxis=(X=-1,Y=0,Z=0)
    DriverItem=(Type=Bone,Name="thigh_l")
    FalloffRegionScaleFactors=(PositiveWidth=1,NegativeWidth=1,PositiveHeight=1,NegativeHeight=1)
    FalloffSize=0.55
    FlipHeightScaling=False
    FlipWidthScaling=False
    OptionalParentItem=(Type=Bone,Name="None")
    RotationOffset=(X=-90,Y=0,Z=20)
    Debug.bDrawDebug=@n7.Value
   )
   > &RigUnit_ModifyTransforms(
    ItemToModify="((Item=(Type=Bone,Name=\"auto_skirt_front_lf\"),Transform=(Rotation=(X=0,Y=0,Z=0,W=-1),Translation=(X=0,Y=0,Z=0),Scale3D=(X=1,Y=1,Z=1))),(Item=(Type=Bone,Name=\"auto_skirt_mid_lf\"),Transform=(Rotation=(X=0,Y=0,Z=0,W=1),Translation=(X=0,Y=0,Z=0),Scale3D=(X=1,Y ...(truncated)"
    WeightMinimum=0
    Weight=@n12.OutputParam
   )
   RigVMFunction_MathBoolToFloat = &RigVMFunction_MathBoolToFloat(Value=RigVMVariableNode() {variable=Sliding}.Value)
   > &RigUnit_ModifyTransforms(
    ItemToModify="((Item=(Type=Bone,Name=\"auto_skirt_midback_rt\"),Transform=(Rotation=(X=0,Y=0.3007,Z=0,W=-0.9537),Translation=(X=0,Y=-3,Z=0),Scale3D=(X=1,Y=1,Z=1))),(Item=(Type=Bone,Name=\"auto_skirt_midback_lf\"),Transform=(Rotation=(X=0,Y=0.3007,Z=0,W=-0.9537),Translation= ...(truncated)"
    WeightMinimum=0
    Weight=@RigVMFunction_MathBoolToFloat.Result
   )
   > &RigUnit_ModifyTransforms(
    ItemToModify="((Item=(Type=Bone,Name=\"auto_skirt_midback_rt\"),Transform=(Rotation=(X=0,Y=-0.0436,Z=0,W=0.999),Translation=(X=0,Y=0,Z=0),Scale3D=(X=1,Y=1,Z=1))),(Item=(Type=Bone,Name=\"auto_skirt_midback_lf\"),Transform=(Rotation=(X=0,Y=0.0436,Z=0,W=-0.999),Translation=(X= ...(truncated)"
    WeightMinimum=0
    Weight=&RigVMFunction_MathBoolToFloat(Value=RigVMVariableNode() {variable=Sprint}.Value).Result
   )
   n13 = RigVMVariableNode() {variable=CrouchingNotMove}
   n14 = /Script/RigVM.RigVMFunction_MathDoubleGreater(B=0.6, A=RigVMVariableNode() {variable=CrouchingButUp}.Value) {is=struct RigVMUnitNode, template="Greater::Execute(in A,in B,out Result)"}
   AlphaInterp = &RigVMFunction_AlphaInterp(
    InterpSpeedDecreasing=1.5
    InterpSpeedIncreasing=2
    Scale=1
    bClampResult=true
    bMapRange=true
    Value=&RigVMFunction_MathDoubleMul(
     B=&RigVMFunction_MathBoolToFloat(Value=@n13.Value).Result
     A=&RigVMFunction_MathBoolToFloat(Value=@n14.Result).Result
    ).Result
   )
   > &RigUnit_ModifyTransforms(
    ItemToModify=((Item=(Type=Bone,Name="base_skirt_front_lf"),Transform=(Rotation=(X=0.2549,Y=0.1677,Z=-0.0449,W=0.9513),Translation=(X=-1.5,Y=0,Z=0),Scale3D=(X=1,Y=1,Z=1))))
    WeightMinimum=0
    Weight=@AlphaInterp.Result
   )
   &RigUnit_ModifyTransforms(
    ItemToModify=((Item=(Type=Bone,Name="dyn_armprotection_lf"),Transform=(Rotation=(X=-0.0008,Y=-0.0436,Z=-0.0174,W=0.9989),Translation=(X=-5,Y=0.1,Z=0.9),Scale3D=(X=0.9,Y=1,Z=1))))
    WeightMinimum=0
    Weight=RigVMDispatchNode(BoneName=root, Default=0, Name=deformRig_l_wrist_up_alpha) {template="DISPATCH_RigDispatch_GetAnimAttribute(in Name,in BoneName,in Default,out Value,out Found)"}.Value
    ExecutePin=@RigVMFunction_Sequence.D
   )
   > &RigUnit_ModifyTransforms(
    ItemToModify=((Item=(Type=Bone,Name="dyn_armprotection_rt"),Transform=(Rotation=(X=-0.0008,Y=-0.0436,Z=-0.0174,W=0.9989),Translation=(X=-5,Y=0.1,Z=0.9),Scale3D=(X=0.9,Y=1,Z=1))))
    WeightMinimum=0
    Weight=RigVMDispatchNode(BoneName=root, Default=0, Name=deformRig_r_wrist_up_alpha) {template="DISPATCH_RigDispatch_GetAnimAttribute(in Name,in BoneName,in Default,out Value,out Found)"}.Value
   )
   &RigUnit_ModifyTransforms(
    ItemToModify=((Item=(Type=Bone,Name="dyn_kneepad_lf"),Transform=(Rotation=(X=0,Y=0,Z=0,W=1),Translation=(X=-3,Y=0,Z=1),Scale3D=(X=1,Y=1,Z=1))))
    WeightMinimum=0
    Weight=RigVMDispatchNode(BoneName=root, Default=0, Name=deformRig_l_knee_alpha) {template="DISPATCH_RigDispatch_GetAnimAttribute(in Name,in BoneName,in Default,out Value,out Found)"}.Value
    ExecutePin=@RigVMFunction_Sequence.E
   )
   > &RigUnit_ModifyTransforms(
    ItemToModify=((Item=(Type=Bone,Name="dyn_kneepad_rt"),Transform=(Rotation=(X=0,Y=0,Z=0,W=1),Translation=(X=-3,Y=0,Z=1),Scale3D=(X=1,Y=1,Z=1))))
    WeightMinimum=0
    Weight=RigVMDispatchNode(BoneName=root, Default=0, Name=deformRig_r_knee_alpha) {template="DISPATCH_RigDispatch_GetAnimAttribute(in Name,in BoneName,in Default,out Value,out Found)"}.Value
   )
   RigUnit_GetRelativeTransformForItem = &RigUnit_GetRelativeTransformForItem(Child=(Type=Bone,Name="pelvis"), Parent=(Type=Bone,Name="root"), bChildInitial=False, bParentInitial=False)
   Remap = &RigVMFunction_MathDoubleRemap(
    SourceMaximum=30
    SourceMinimum=60
    TargetMaximum=0
    TargetMinimum=1
    bClamp=False
    Value=@RigUnit_GetRelativeTransformForItem.RelativeTransform.Translation.Z
   )
   n15 = &RigVMFunction_AnimEvalRichCurve(
    Curve="(EditorCurveData=(Keys=((InterpMode=RCIM_Cubic,TangentMode=RCTM_Break,TangentWeightMode=RCTWM_WeightedNone,Time=0,Value=0,ArriveTangent=0,ArriveTangentWeight=0,LeaveTangent=0,LeaveTangentWeight=0),(InterpMode=RCIM_Cubic,TangentMode=RCTM_Break,TangentWeight ...(truncated)"
    SourceMinimum=0
    TargetMinimum=0
    Value=@Remap.Result
   )
   RigVMVariableNode(Value=@n15.Result, ExecuteContext=@RigVMFunction_Sequence.C) {variable=CrouchingButUp}
   n16 = &RigUnit_ModifyTransforms(
    ItemToModify=((Item=(Type=Bone,Name="dyn_shoulderpad_lf"),Transform=(Rotation=(X=0,Y=0,Z=0,W=1),Translation=(X=0,Y=0,Z=1),Scale3D=(X=1,Y=1,Z=1))))
    Weight=1
    WeightMinimum=0
    ExecutePin=@RigVMFunction_Sequence.F
   )
   RigVMFunction_ControlFlowBranch = &RigVMFunction_ControlFlowBranch(ExecuteContext=@n16.ExecutePin, Condition=RigVMVariableNode() {variable=OnAir}.Value)
   n17 = &RigVMFunction_MathBoolMake(Value=false)
   n18 = &RigUnit_SphericalPoseReader(
    ActiveRegionScaleFactors=(PositiveWidth=2,NegativeWidth=2,PositiveHeight=2,NegativeHeight=2)
    ActiveRegionSize=0.01
    Debug=(bDrawDebug=true,bDraw2D=false,DebugScale=25,DebugSegments=20,DebugThickness=0.25)
    DriverAxis=(X=1,Y=0,Z=0)
    DriverItem=(Type=Bone,Name="upperarm_l")
    FalloffRegionScaleFactors=(PositiveWidth=1,NegativeWidth=1,PositiveHeight=1,NegativeHeight=1)
    FalloffSize=0.3
    OptionalParentItem=(Type=Bone,Name="clavicle_l")
    RotationOffset=(X=85,Y=0,Z=0)
    ExecutePin=@RigVMFunction_ControlFlowBranch.False
    Debug.bDrawDebug=@n17.Value
   )
   > &RigUnit_ModifyTransforms(
    ItemToModify=((Item=(Type=Bone,Name="dyn_shoulderpad_lf"),Transform=(Rotation=(X=0.2154,Y=-0.0207,Z=-0.0936,W=0.9718),Translation=(X=0,Y=0,Z=0),Scale3D=(X=1,Y=1,Z=1))))
    WeightMinimum=0
    Weight=@n18.OutputParam
   )
   > n19 = &RigUnit_SphericalPoseReader(
    ActiveRegionScaleFactors=(PositiveWidth=2,NegativeWidth=2,PositiveHeight=2,NegativeHeight=2)
    ActiveRegionSize=0.03
    Debug=(bDrawDebug=true,bDraw2D=false,DebugScale=25,DebugSegments=20,DebugThickness=0.25)
    DriverAxis=(X=1,Y=0,Z=0)
    DriverItem=(Type=Bone,Name="upperarm_l")
    FalloffRegionScaleFactors=(PositiveWidth=1,NegativeWidth=1,PositiveHeight=1,NegativeHeight=1)
    FalloffSize=0.2
    OptionalParentItem=(Type=Bone,Name="clavicle_l")
    RotationOffset=(X=10,Y=-35,Z=0)
    Debug.bDrawDebug=@n17.Value
   )
   > &RigUnit_ModifyTransforms(
    ItemToModify=((Item=(Type=Bone,Name="dyn_shoulderpad_lf"),Transform=(Rotation=(X=0,Y=-0.0175,Z=0,W=0.9998),Translation=(X=7,Y=0,Z=0),Scale3D=(X=1,Y=1,Z=1))))
    WeightMinimum=0
    Weight=@n19.OutputParam
   )
   > n20 = &RigUnit_SphericalPoseReader(
    ActiveRegionScaleFactors=(PositiveWidth=2,NegativeWidth=2,PositiveHeight=2,NegativeHeight=2)
    ActiveRegionSize=0.03
    Debug=(bDrawDebug=true,bDraw2D=false,DebugScale=25,DebugSegments=20,DebugThickness=0.25)
    DriverAxis=(X=1,Y=0,Z=0)
    DriverItem=(Type=Bone,Name="upperarm_l")
    FalloffRegionScaleFactors=(PositiveWidth=1,NegativeWidth=1,PositiveHeight=1,NegativeHeight=1)
    FalloffSize=0.3
    OptionalParentItem=(Type=Bone,Name="clavicle_l")
    RotationOffset=(X=-10,Y=-20,Z=0)
    Debug.bDrawDebug=@n17.Value
   )
   > &RigUnit_ModifyTransforms(
    ItemToModify=((Item=(Type=Bone,Name="dyn_shoulderpad_lf"),Transform=(Rotation=(X=-0.0019,Y=-0.0436,Z=-0.0436,W=0.9981),Translation=(X=7,Y=0,Z=0),Scale3D=(X=1,Y=1,Z=1))))
    WeightMinimum=0
    Weight=@n20.OutputParam
   )
   > n21 = &RigUnit_SphericalPoseReader(
    ActiveRegionScaleFactors=(PositiveWidth=2,NegativeWidth=2,PositiveHeight=2,NegativeHeight=2)
    ActiveRegionSize=0.03
    Debug=(bDrawDebug=true,bDraw2D=false,DebugScale=25,DebugSegments=20,DebugThickness=0.25)
    DriverAxis=(X=1,Y=0,Z=0)
    DriverItem=(Type=Bone,Name="upperarm_l")
    FalloffRegionScaleFactors=(PositiveWidth=1,NegativeWidth=1,PositiveHeight=1,NegativeHeight=1)
    FalloffSize=0.3
    OptionalParentItem=(Type=Bone,Name="clavicle_l")
    RotationOffset=(X=0,Y=-110,Z=20)
    Debug.bDrawDebug=@n17.Value
   )
   > &RigUnit_ModifyTransforms(
    ItemToModify=((Item=(Type=Bone,Name="dyn_shoulderpad_lf"),Transform=(Rotation=(X=0.1736,Y=0,Z=0,W=0.9848),Translation=(X=0,Y=-2,Z=0),Scale3D=(X=1,Y=1,Z=1))))
    WeightMinimum=0
    Weight=@n21.OutputParam
   )
   > n22 = &RigUnit_SphericalPoseReader(
    ActiveRegionScaleFactors=(PositiveWidth=2,NegativeWidth=2,PositiveHeight=2,NegativeHeight=2)
    ActiveRegionSize=0.03
    Debug=(bDrawDebug=true,bDraw2D=false,DebugScale=25,DebugSegments=20,DebugThickness=0.25)
    DriverAxis=(X=1,Y=0,Z=0)
    DriverItem=(Type=Bone,Name="upperarm_l")
    FalloffRegionScaleFactors=(PositiveWidth=1,NegativeWidth=1,PositiveHeight=1,NegativeHeight=1)
    FalloffSize=0.3
    OptionalParentItem=(Type=Bone,Name="clavicle_l")
    RotationOffset=(X=-10,Y=-85,Z=20)
    Debug.bDrawDebug=@n17.Value
   )
   > &RigUnit_ModifyTransforms(
    ItemToModify=((Item=(Type=Bone,Name="dyn_shoulderpad_lf"),Transform=(Rotation=(X=-0.0872,Y=0,Z=0,W=0.9962),Translation=(X=3,Y=1.5,Z=0),Scale3D=(X=1,Y=1,Z=1))))
    WeightMinimum=0
    Weight=@n22.OutputParam
   )
   > n23 = &RigUnit_SphericalPoseReader(
    ActiveRegionScaleFactors=(PositiveWidth=2,NegativeWidth=2,PositiveHeight=2,NegativeHeight=2)
    ActiveRegionSize=0.03
    Debug=(bDrawDebug=true,bDraw2D=false,DebugScale=25,DebugSegments=20,DebugThickness=0.25)
    DriverAxis=(X=1,Y=0,Z=0)
    DriverItem=(Type=Bone,Name="upperarm_l")
    FalloffRegionScaleFactors=(PositiveWidth=1,NegativeWidth=1,PositiveHeight=1,NegativeHeight=1)
    FalloffSize=0.3
    OptionalParentItem=(Type=Bone,Name="clavicle_l")
    RotationOffset=(X=70,Y=0,Z=-25)
    Debug.bDrawDebug=@n17.Value
   )
   > &RigUnit_ModifyTransforms(
    ItemToModify=((Item=(Type=Bone,Name="dyn_shoulderpad_lf"),Transform=(Rotation=(X=-0.0958,Y=-0.0017,Z=0.0174,W=0.9952),Translation=(X=5,Y=1.5,Z=0),Scale3D=(X=1,Y=1,Z=1))))
    WeightMinimum=0
    Weight=@n23.OutputParam
   )
   > n24 = &RigUnit_SphericalPoseReader(
    ActiveRegionScaleFactors=(PositiveWidth=2,NegativeWidth=2,PositiveHeight=2,NegativeHeight=2)
    ActiveRegionSize=0.03
    Debug=(bDrawDebug=true,bDraw2D=false,DebugScale=25,DebugSegments=20,DebugThickness=0.25)
    DriverAxis=(X=1,Y=0,Z=0)
    DriverItem=(Type=Bone,Name="upperarm_l")
    FalloffRegionScaleFactors=(PositiveWidth=1,NegativeWidth=1,PositiveHeight=1,NegativeHeight=1)
    FalloffSize=0.15
    OptionalParentItem=(Type=Bone,Name="clavicle_l")
    RotationOffset=(X=45,Y=-40,Z=-25)
    Debug.bDrawDebug=@n17.Value
   )
   > &RigUnit_ModifyTransforms(
    ItemToModify=((Item=(Type=Bone,Name="dyn_shoulderpad_lf"),Transform=(Rotation=(X=0,Y=-0.0175,Z=0,W=0.9998),Translation=(X=1,Y=0,Z=0),Scale3D=(X=1,Y=1,Z=1))))
    WeightMinimum=0
    Weight=@n24.OutputParam
   )
   > n25 = &RigUnit_SphericalPoseReader(
    ActiveRegionScaleFactors=(PositiveWidth=2,NegativeWidth=2,PositiveHeight=2,NegativeHeight=2)
    ActiveRegionSize=0.03
    Debug=(bDrawDebug=true,bDraw2D=false,DebugScale=25,DebugSegments=20,DebugThickness=0.25)
    DriverAxis=(X=1,Y=0,Z=0)
    DriverItem=(Type=Bone,Name="upperarm_l")
    FalloffRegionScaleFactors=(PositiveWidth=1,NegativeWidth=1,PositiveHeight=1,NegativeHeight=1)
    FalloffSize=0.3
    OptionalParentItem=(Type=Bone,Name="clavicle_l")
    RotationOffset=(X=0,Y=0,Z=0)
    Debug.bDrawDebug=@n17.Value
   )
   Greater = &RigVMFunction_MathDoubleGreater(B=0.6, A=RigVMVariableNode() {variable=CrouchingButUp}.Value) {template="Greater::Execute(in A,in B,out Result)"}
   Multiply_2 = RigVMAggregateNode(
    C=&RigVMFunction_MathBoolToFloat(Value=RigVMVariableNode() {variable=CrouchingNotMove}.Value).Result
    B=&RigVMFunction_MathBoolToFloat(Value=@Greater.Result).Result
    A=@n25.OutputParam
   )
   > &RigUnit_ModifyTransforms(
    ItemToModify=((Item=(Type=Bone,Name="dyn_shoulderpad_lf"),Transform=(Rotation=(X=0.1736,Y=0,Z=0,W=0.9848),Translation=(X=6,Y=0,Z=0),Scale3D=(X=1,Y=1,Z=1))))
    WeightMinimum=0
    Weight=@Multiply_2.Result
   )
   > DeltaRotation = RigVMFunctionReferenceNode(
    Rotation Order=XYZ
    child=upperarm_l
    debugRotation=false
    max=0
    min=-20
    onlyTwist=true
    parent=upperarm_l_null
    parentType=Null
    premult=1
    primAxis=0
    twistAxis=(X=1,Y=0,Z=0)
   )
   > &RigUnit_ModifyTransforms(
    ItemToModify=((Item=(Type=Bone,Name="dyn_shoulderpad_lf"),Transform=(Rotation=(X=-0.2588,Y=0,Z=0,W=0.9659),Translation=(X=0,Y=3,Z=0),Scale3D=(X=1,Y=1,Z=1))))
    WeightMinimum=0
    Weight=@DeltaRotation.outputWeight
   )
   &RigUnit_ModifyTransforms(
    ItemToModify=((Item=(Type=Bone,Name="dyn_shoulderpad_lf"),Transform=(Rotation=(X=0,Y=0,Z=0,W=1),Translation=(X=6,Y=0,Z=1),Scale3D=(X=1,Y=1,Z=1))))
    Weight=1
    WeightMinimum=0
    ExecutePin=@RigVMFunction_ControlFlowBranch.True
   )
  ]}
]
hierarchy [
 root {loc=<0 0 0>, rot=<0 0 0 1>}
 pelvis {parent=root, loc=<0 -0.7071 93.1191>, rot=<0.0127 0.707 -0.0127 -0.707>}
 spine_01 {parent=pelvis, loc=<6.7494 -0.6388 0>, rot=<0 0 0.034 -0.9994>}
 spine_02 {parent=spine_01, loc=<7.2518 -0.035 0>, rot=<0 0 0.0238 -0.9997>}
 spine_03 {parent=spine_02, loc=<6.0104 0 0>, rot=<0 0 0.0149 -0.9999>}
 spine_04 {parent=spine_03, loc=<8.1024 0 0>, rot=<0 0 -0.0132 -0.9999>}
 spine_05 {parent=spine_04, loc=<10.0146 0 0>, rot=<0 0 -0.0015 -1>}
 neck_01 {parent=spine_05, loc=<11.3138 0.1529 0>, rot=<0 0 -0.2003 -0.9797>}
 neck_02 {parent=neck_01, loc=<4.5627 0 0>, rot=<0 0 0.0504 -0.9987>}
 head {parent=neck_02, loc=<5.3914 0 0>, rot=<0 0 0.115 -0.9934>}
 attach_cape {parent=spine_05, loc=<9.017 -11.714 0>, rot=<0.0616 -0.7044 -0.0616 -0.7044>}
 cape {parent=attach_cape, type=Null, loc=<0 0 0>, rot=<0 0 0 1>}
 attach_backpack {parent=spine_05, loc=<-0.123 -13.473 0>, rot=<-0.0093 -0.707 0.0093 -0.707>}
 Backpack_BR {parent=attach_backpack, type=Null, loc=<0 0 0>, rot=<0 0 0 1>}
 clavicle_l {parent=spine_05, loc=<4.3575 -0.1466 -2.3223>, rot=<0.0465 -0.7289 -0.009 -0.6829>}
 upperarm_l {parent=clavicle_l, loc=<15.5111 0 0>, rot=<0.0226 -0.3234 -0.0088 -0.9459>}
 lowerarm_l {parent=upperarm_l, loc=<26.6514 0 0>, rot=<0.0391 -0.0099 -0.2453 -0.9686>}
 hand_l {parent=lowerarm_l, loc=<22.3491 0 0>, rot=<-0.7081 -0.0507 0.0975 -0.6975>}
 weapon_l {parent=hand_l, loc=<-0.1974 -3.4111 0.3811>, rot=<0 0 0 1>}
 thumb_01_l {parent=hand_l, loc=<1.3118 -0.3191 -2.459>, rot=<0.6546 -0.4042 0.0664 -0.6354>}
 thumb_02_l {parent=thumb_01_l, loc=<5.9336 -0.0448 0.0401>, rot=<-0.0037 0.0213 0.298 -0.9543>}
 thumb_03_l {parent=thumb_02_l, loc=<2.2223 0.3682 -0.0907>, rot=<0.0118 -0.013 0.0066 -0.9998>}
 index_metacarpal_l {parent=hand_l, loc=<1.6535 0.0183 -2.96>, rot=<0.0158 -0.0124 -0.0173 -0.9996>}
 index_01_l {parent=index_metacarpal_l, loc=<8.3008 0.0357 -0.0524>, rot=<-0.0044 0.0485 0.0924 -0.9945>}
 index_02_l {parent=index_01_l, loc=<3.3587 -0.3004 0.0452>, rot=<-0.002 0.0027 0.3225 -0.9466>}
 index_03_l {parent=index_02_l, loc=<2.5578 -0.123 -0.0252>, rot=<-0.002 0.0067 0.2558 -0.9667>}
 middle_metacarpal_l {parent=hand_l, loc=<2.1856 -0.1725 -0.5984>, rot=<0.006 0.0168 -0.0329 -0.9993>}
 middle_01_l {parent=middle_metacarpal_l, loc=<7.3275 -0.2447 0.0129>, rot=<-0.0208 0.0042 0.0858 -0.9961>}
 middle_02_l {parent=middle_01_l, loc=<4.0123 0.1401 0.1175>, rot=<-0.0218 0.0258 0.3512 -0.9357>}
 middle_03_l {parent=middle_02_l, loc=<2.6077 0.0534 0.1373>, rot=<0.0082 0.0188 0.1708 -0.9851>}
 ring_metacarpal_l {parent=hand_l, loc=<2.2074 -0.3625 1.3638>, rot=<-0.0284 0.0722 -0.0458 -0.9959>}
 ring_01_l {parent=ring_metacarpal_l, loc=<7.0295 -0.6197 -0.1446>, rot=<-0.0489 -0.0252 0.1892 -0.9804>}
 ring_02_l {parent=ring_01_l, loc=<3.6174 0.4647 -0.1027>, rot=<0 0 0.2806 -0.9598>}
 ring_03_l {parent=ring_02_l, loc=<2.4152 0.114 0.0199>, rot=<0 0 0.1189 -0.9929>}
 pinky_metacarpal_l {parent=hand_l, loc=<1.7616 -0.9753 3.3814>, rot=<-0.0397 0.0814 -0.069 -0.9935>}
 pinky_01_l {parent=pinky_metacarpal_l, loc=<7.1709 -0.4831 -0.1094>, rot=<-0.0654 -0.0035 0.2242 -0.9723>}
 pinky_02_l {parent=pinky_01_l, loc=<2.8274 0.0893 -0.1043>, rot=<0.0037 0.0011 0.2854 -0.9584>}
 pinky_03_l {parent=pinky_02_l, loc=<2.2046 0.0307 -0.0078>, rot=<-0.0002 0.0098 0.1445 -0.9895>}
 hand_attach_l {parent=hand_l, loc=<8.5 -3 0>, rot=<0.7071 0.7071 0 0>}
 deform_wrist_dn_base_l {parent=hand_l, loc=<0 -2.2342 0>, rot=<0 0 0 -1>}
 deform_wrist_dn_l {parent=deform_wrist_dn_base_l, loc=<0 0 0>, rot=<0 0 0 1>}
 deform_wrist_up_base_l {parent=hand_l, loc=<0 2.2682 0>, rot=<0 0 0 -1>}
 deform_wrist_up_l {parent=deform_wrist_up_base_l, loc=<0 0 0>, rot=<0 0 0 1>}
 lowerarm_twist_02_l {parent=lowerarm_l, loc=<9.6199 0 0>, rot=<0 0 0 -1>}
 base_armprotection_lf {parent=lowerarm_twist_02_l, loc=<1.5167 -0.05 3.6516>, rot=<-0.013 -0.0191 0.0262 -0.9994>}
 dyn_armprotection_lf {parent=base_armprotection_lf, loc=<0 0 0>, rot=<0 0 0 1>}
 lowerarm_twist_01_l {parent=lowerarm_l, loc=<17.8549 0 0>, rot=<0 0 0 -1>}
 deform_elbow_in_base_l {parent=lowerarm_l, loc=<0.2112 5.259 0.8274>, rot=<0 0 0 1>}
 deform_elbow_in_l {parent=deform_elbow_in_base_l, loc=<0 0 0>, rot=<0 0 0 1>}
 deform_elbow_out_base_l {parent=lowerarm_l, loc=<0 -4 0>, rot=<0 0 0 1>}
 deform_elbow_out_l {parent=deform_elbow_out_base_l, loc=<0 0 0>, rot=<0 0 0 1>}
 upperarm_twist_01_l {parent=upperarm_l, loc=<0.0041 0 0>, rot=<0 0 0 -1>}
 deform_upperarm_bk_base_l {parent=upperarm_twist_01_l, loc=<1.0813 -7.2193 -1.9453>, rot=<0 0 0 1>}
 deform_upperarm_bk_l {parent=deform_upperarm_bk_base_l, loc=<0 0 0>, rot=<0 0 0 1>}
 deform_upperarm_fr_base_l {parent=upperarm_twist_01_l, loc=<-0.3691 7.0506 -2.765>, rot=<0 0 0 1>}
 deform_upperarm_fr_l {parent=deform_upperarm_fr_base_l, loc=<0 0 0>, rot=<0 0 0 1>}
 deform_upperarm_up_base_l {parent=upperarm_twist_01_l, loc=<-3.2216 -1.1391 3.3311>, rot=<0 0 0 1>}
 deform_upperarm_up_l {parent=deform_upperarm_up_base_l, loc=<0 0 0>, rot=<0 0 0 1>}
 upperarm_twist_02_l {parent=upperarm_l, loc=<13.7941 0 0>, rot=<0 0 0 -1>}
 base_shoulderpad_lf {parent=upperarm_l, loc=<-5.3395 -1.6253 6.5164>, rot=<-0.0072 -0.0179 -0.0179 -0.9997>}
 dyn_shoulderpad_lf {parent=base_shoulderpad_lf, loc=<0 0 0>, rot=<0 0 0 1>}
 upperarm_l_null {parent=clavicle_l, type=Null, loc=<15.5111 0 0>, rot=<-0.4534 -0.2845 0.1541 -0.8305>}
 clavicle_r {parent=spine_05, loc=<4.3575 -0.1466 2.3223>, rot=<0.7289 0.0465 -0.6829 0.009>}
 upperarm_r {parent=clavicle_r, loc=<-15.5111 0 0>, rot=<0.0226 -0.3234 -0.0088 -0.9459>}
 lowerarm_r {parent=upperarm_r, loc=<-26.6514 0 0>, rot=<0.0391 -0.0099 -0.2453 -0.9686>}
 hand_r {parent=lowerarm_r, loc=<-22.3491 0 0>, rot=<-0.7081 -0.0507 0.0975 -0.6975>}
 weapon_r {parent=hand_r, loc=<0.1974 3.4111 -0.3811>, rot=<0 0 0 1>}
 thumb_01_r {parent=hand_r, loc=<-1.3118 0.3191 2.459>, rot=<0.6546 -0.4042 0.0664 -0.6354>}
 thumb_02_r {parent=thumb_01_r, loc=<-5.9336 0.0448 -0.0401>, rot=<-0.0037 0.0213 0.298 -0.9543>}
 thumb_03_r {parent=thumb_02_r, loc=<-2.2223 -0.3682 0.0907>, rot=<0.0118 -0.013 0.0066 -0.9998>}
 index_metacarpal_r {parent=hand_r, loc=<-1.6535 -0.0183 2.96>, rot=<0.0158 -0.0124 -0.0173 -0.9996>}
 index_01_r {parent=index_metacarpal_r, loc=<-8.3008 -0.0357 0.0524>, rot=<-0.0044 0.0485 0.0924 -0.9945>}
 index_02_r {parent=index_01_r, loc=<-3.3587 0.3004 -0.0452>, rot=<-0.002 0.0027 0.3225 -0.9466>}
 index_03_r {parent=index_02_r, loc=<-2.5578 0.123 0.0252>, rot=<-0.002 0.0067 0.2558 -0.9667>}
 middle_metacarpal_r {parent=hand_r, loc=<-2.1856 0.1725 0.5984>, rot=<0.006 0.0168 -0.0329 -0.9993>}
 middle_01_r {parent=middle_metacarpal_r, loc=<-7.3275 0.2447 -0.0129>, rot=<-0.0208 0.0042 0.0858 -0.9961>}
 middle_02_r {parent=middle_01_r, loc=<-4.0123 -0.1401 -0.1175>, rot=<-0.0218 0.0258 0.3512 -0.9357>}
 middle_03_r {parent=middle_02_r, loc=<-2.6077 -0.0534 -0.1373>, rot=<0.0082 0.0188 0.1708 -0.9851>}
 ring_metacarpal_r {parent=hand_r, loc=<-2.2074 0.3625 -1.3638>, rot=<-0.0284 0.0722 -0.0458 -0.9959>}
 ring_01_r {parent=ring_metacarpal_r, loc=<-7.0295 0.6197 0.1446>, rot=<-0.0489 -0.0252 0.1892 -0.9804>}
 ring_02_r {parent=ring_01_r, loc=<-3.6174 -0.4647 0.1027>, rot=<0 0 0.2806 -0.9598>}
 ring_03_r {parent=ring_02_r, loc=<-2.4152 -0.114 -0.0199>, rot=<0 0 0.1189 -0.9929>}
 pinky_metacarpal_r {parent=hand_r, loc=<-1.7616 0.9753 -3.3814>, rot=<-0.0397 0.0814 -0.069 -0.9935>}
 pinky_01_r {parent=pinky_metacarpal_r, loc=<-7.1709 0.4831 0.1094>, rot=<-0.0654 -0.0035 0.2242 -0.9723>}
 pinky_02_r {parent=pinky_01_r, loc=<-2.8274 -0.0893 0.1043>, rot=<0.0037 0.0011 0.2854 -0.9584>}
 pinky_03_r {parent=pinky_02_r, loc=<-2.2046 -0.0307 0.0078>, rot=<-0.0002 0.0098 0.1445 -0.9895>}
 hand_attach_r {parent=hand_r, loc=<-8.5001 2.9999 0>, rot=<0 0 -0.7071 -0.7071>}
 deform_wrist_dn_base_r {parent=hand_r, loc=<-0.0004 2.2345 -0.0001>, rot=<0 0 0 -1>}
 deform_wrist_dn_r {parent=deform_wrist_dn_base_r, loc=<0 0 0>, rot=<0 0 0 1>}
 deform_wrist_up_base_r {parent=hand_r, loc=<0.0003 -2.2684 0.0001>, rot=<0 0 0 -1>}
 deform_wrist_up_r {parent=deform_wrist_up_base_r, loc=<0 0 0>, rot=<0 0 0 1>}
 lowerarm_twist_02_r {parent=lowerarm_r, loc=<-9.6199 0 0>, rot=<0 0 0 -1>}
 base_armprotection_rt {parent=lowerarm_twist_02_r, loc=<-1.5167 0.05 -3.6516>, rot=<-0.0262 -0.9994 -0.013 0.0191>}
 dyn_armprotection_rt {parent=base_armprotection_rt, loc=<0 0 0>, rot=<0 0 0 1>}
 lowerarm_twist_01_r {parent=lowerarm_r, loc=<-17.8549 0 0>, rot=<0 0 0 -1>}
 deform_elbow_in_base_r {parent=lowerarm_r, loc=<-0.2111 -5.2591 -0.8275>, rot=<0 0 0 1>}
 deform_elbow_in_r {parent=deform_elbow_in_base_r, loc=<-0.0001 0 0.0002>, rot=<0 0 0 1>}
 deform_elbow_out_base_r {parent=lowerarm_r, loc=<0 4 -0.0001>, rot=<0 0 0 1>}
 deform_elbow_out_r {parent=deform_elbow_out_base_r, loc=<0 0 0>, rot=<0 0 0 1>}
 upperarm_twist_01_r {parent=upperarm_r, loc=<-0.0041 0 0>, rot=<0 0 0 -1>}
 deform_upperarm_bk_base_r {parent=upperarm_twist_01_r, loc=<-1.0814 7.2192 1.9455>, rot=<0 0 0 1>}
 deform_upperarm_bk_r {parent=deform_upperarm_bk_base_r, loc=<0 0 0>, rot=<0 0 0 1>}
 deform_upperarm_fr_base_r {parent=upperarm_twist_01_r, loc=<0.3689 -7.0506 2.7652>, rot=<0 0 0 1>}
 deform_upperarm_fr_r {parent=deform_upperarm_fr_base_r, loc=<0 0 0>, rot=<0 0 0 1>}
 deform_upperarm_up_base_r {parent=upperarm_twist_01_r, loc=<3.2214 1.1391 -3.3308>, rot=<0 0 0 1>}
 deform_upperarm_up_r {parent=deform_upperarm_up_base_r, loc=<0 0 0>, rot=<0 0 0 -1>}
 upperarm_twist_02_r {parent=upperarm_r, loc=<-13.7941 0 0>, rot=<0 0 0 -1>}
 clavicle_l_null {parent=spine_05, type=Null, loc=<4.3575 -0.1466 -2.3223>, rot=<0.0465 -0.7289 -0.009 -0.6829>}
 deform_pec_aim_base_l {parent=spine_04, loc=<7.8732 12.6388 -2.4123>, rot=<0.0102 -0.4602 0.3979 -0.7936>}
 deform_pec_aim_l {parent=deform_pec_aim_base_l, loc=<0 0 0>, rot=<0 0 0 1>}
 deform_pec_base_l {parent=deform_pec_aim_l, loc=<10.766 0 0>, rot=<0 0 0 1>}
 deform_pec_l {parent=deform_pec_base_l, loc=<0.055 0 0>, rot=<0 0 0 1>}
 deform_pec_aim_base_r {parent=spine_04, loc=<7.8727 12.6388 2.4123>, rot=<-0.4602 -0.0102 0.7936 0.3979>}
 deform_pec_aim_r {parent=deform_pec_aim_base_r, loc=<0 0 0>, rot=<0 0 0 1>}
 deform_pec_base_r {parent=deform_pec_aim_r, loc=<-10.7662 -0.0006 -0.0007>, rot=<0 0 0 1>}
 deform_pec_r {parent=deform_pec_base_r, loc=<0 0 0>, rot=<0 0 0 1>}
 thigh_l {parent=pelvis, loc=<-2.5693 -0.0109 -7.7501>, rot=<0.0012 0.0572 0.0155 -0.9982>}
 calf_l {parent=thigh_l, loc=<-39.0484 0 0>, rot=<0.0006 -0.0161 -0.0391 -0.9991>}
 foot_l {parent=calf_l, loc=<-42.561 0.1077 0.4292>, rot=<0.0418 -0.0455 0.0343 -0.9975>}
 ball_l {parent=foot_l, loc=<-6.1599 13.8264 0.6806>, rot=<0.0001 -0.0001 -0.7063 -0.7079>}
 calf_twist_02_l {parent=calf_l, loc=<-18.8064 -0.0967 0.0397>, rot=<0.0237 0.0095 -0.0035 -0.9997>}
 calf_twist_01_l {parent=calf_l, loc=<-35.813 -0.0967 0.0397>, rot=<0.0237 0.0095 -0.0035 -0.9997>}
 deform_knee_in_base_l {parent=calf_l, loc=<1.298 -5.2957 -0.0496>, rot=<0 0 0 -1>}
 deform_knee_in_l {parent=deform_knee_in_base_l, loc=<0 0 0>, rot=<0 0 0 1>}
 deform_knee_out_base_l {parent=calf_l, loc=<0 4 0>, rot=<0 0 0 -1>}
 deform_knee_out_l {parent=deform_knee_out_base_l, loc=<0 0 0>, rot=<0 0 0 1>}
 base_kneepad_lf {parent=calf_l, loc=<-12.4352 5.8896 -0.5963>, rot=<0.0397 -0.7233 -0.6883 0.0397>}
 dyn_kneepad_lf {parent=base_kneepad_lf, loc=<0 0 0>, rot=<0 0 0 1>}
 thigh_twist_01_l {parent=thigh_l, loc=<-21.6303 0 0>, rot=<0.0013 0 0 -1>}
 deform_groin_base_l {parent=thigh_l, loc=<-2.1908 8.1142 -1.7442>, rot=<0 0 0 -1>}
 deform_groin_l {parent=deform_groin_base_l, loc=<0 0 0>, rot=<0 0 0 1>}
 thigh_r {parent=pelvis, loc=<-2.5693 -0.0109 7.7501>, rot=<0.0572 -0.0012 0.9982 0.0155>}
 calf_r {parent=thigh_r, loc=<39.0484 0 0>, rot=<0.0006 -0.0161 -0.0391 -0.9991>}
 foot_r {parent=calf_r, loc=<42.561 -0.1077 -0.4292>, rot=<0.0418 -0.0455 0.0343 -0.9975>}
 ball_r {parent=foot_r, loc=<6.1599 -13.8264 -0.6806>, rot=<0.0001 -0.0001 -0.7063 -0.7079>}
 calf_twist_02_r {parent=calf_r, loc=<18.8064 0.0967 -0.0397>, rot=<0.0237 0.0095 -0.0035 -0.9997>}
 calf_twist_01_r {parent=calf_r, loc=<35.813 0.0967 -0.0397>, rot=<0.0237 0.0095 -0.0035 -0.9997>}
 deform_knee_in_base_r {parent=calf_r, loc=<-1.298 5.2957 0.0496>, rot=<0 0 0 1>}
 deform_knee_in_r {parent=deform_knee_in_base_r, loc=<0 0 0>, rot=<0 0 0 1>}
 deform_knee_out_base_r {parent=calf_r, loc=<0 -4 0>, rot=<0 0 0 -1>}
 deform_knee_out_r {parent=deform_knee_out_base_r, loc=<0 0 0>, rot=<0 0 0 1>}
 base_kneepad_rt {parent=calf_r, loc=<12.4352 -5.8896 0.5963>, rot=<-0.6647 -0.0384 -0.041 -0.745>}
 dyn_kneepad_rt {parent=base_kneepad_rt, loc=<0 0 0>, rot=<0 0 0 1>}
 thigh_twist_01_r {parent=thigh_r, loc=<21.6303 0 0>, rot=<0.0013 0 0 -1>}
 deform_groin_base_r {parent=thigh_r, loc=<2.1908 -8.1142 1.7442>, rot=<0 0 0 -1>}
 deform_groin_r {parent=deform_groin_base_r, loc=<0 0 0>, rot=<0 0 0 1>}
 deform_glute_base_l {parent=pelvis, loc=<-8.8064 -8.8222 -8.4287>, rot=<0.0275 0.0501 -0.4472 -0.8926>}
 deform_glute_l {parent=deform_glute_base_l, loc=<0 0 0>, rot=<0 0 0 1>}
 deform_glute_base_r {parent=pelvis, loc=<-8.8062 -8.8222 8.4287>, rot=<0.0501 -0.0275 0.8926 -0.4472>}
 deform_glute_r {parent=deform_glute_base_r, loc=<0 0 0>, rot=<0 0 0 1>}
 base_skirt_back_rt {parent=pelvis, loc=<3.1375 -10.0067 8.765>, rot=<0.09 -0.4822 0.8706 -0.0389>}
 auto_skirt_back_rt {parent=base_skirt_back_rt, loc=<0 0 0>, rot=<0 0 0 1>}
 offset_skirt_back_rt {parent=auto_skirt_back_rt, loc=<0 0 0>, rot=<0 0 0 1>}
 dyn_skirt_back_rt_01 {parent=offset_skirt_back_rt, loc=<0 0 0>, rot=<0 0 0 1>}
 dyn_skirt_back_rt_02 {parent=dyn_skirt_back_rt_01, loc=<9.7989 -0.1955 -0.1228>, rot=<0.0339 0.0054 -0.008 -0.9994>}
 dyn_skirt_back_rt_03 {parent=dyn_skirt_back_rt_02, loc=<11.9686 0 0>, rot=<0.0421 0.0038 -0.0103 -0.9991>}
 dyn_skirt_back_rt_04 {parent=dyn_skirt_back_rt_03, loc=<9.46 0 0>, rot=<-0.0279 -0.0031 0.0069 -0.9996>}
 base_skirt_mid_lf {parent=pelvis, loc=<2.0851 6.1709 -13.7542>, rot=<0.0799 -0.9768 -0.0492 -0.1927>}
 auto_skirt_mid_lf {parent=base_skirt_mid_lf, loc=<0 0 0>, rot=<0 0 0 1>}
 offset_skirt_mid_lf {parent=auto_skirt_mid_lf, loc=<0 0 0>, rot=<0 0 0 1>}
 dyn_skirt_mid_lf_01 {parent=offset_skirt_mid_lf, loc=<0 0 0>, rot=<0 0 0 1>}
 dyn_skirt_mid_lf_02 {parent=dyn_skirt_mid_lf_01, loc=<9.2364 0 0>, rot=<-0.0403 -0.0365 -0.0013 -0.9985>}
 dyn_skirt_mid_lf_03 {parent=dyn_skirt_mid_lf_02, loc=<11.7292 0 0>, rot=<0.0124 0.0075 -0.0013 -0.9999>}
 dyn_skirt_mid_lf_04 {parent=dyn_skirt_mid_lf_03, loc=<9.2852 0 0>, rot=<-0.0519 0.0039 0.0274 -0.9983>}
 base_skirt_midback_rt {parent=pelvis, loc=<2.5085 -5.145 13.5197>, rot=<0.1361 -0.2838 0.9488 -0.0249>}
 auto_skirt_midback_rt {parent=base_skirt_midback_rt, loc=<0 0 0>, rot=<0 0 0 1>}
 offset_skirt_midback_rt {parent=auto_skirt_midback_rt, loc=<0 0 0>, rot=<0 0 0 1>}
 dyn_skirt_midback_rt_01 {parent=offset_skirt_midback_rt, loc=<0 0 0>, rot=<0 0 0 1>}
 dyn_skirt_midback_rt_02 {parent=dyn_skirt_midback_rt_01, loc=<9.5601 0 0>, rot=<0.0133 -0.0151 -0.0038 -0.9998>}
 dyn_skirt_midback_rt_03 {parent=dyn_skirt_midback_rt_02, loc=<11.9238 0 0>, rot=<0.0198 0.0084 -0.0056 -0.9998>}
 dyn_skirt_midback_rt_04 {parent=dyn_skirt_midback_rt_03, loc=<9.514 0 0>, rot=<0.0056 0.0149 -0.0017 -0.9999>}
 base_skirt_midback_lf {parent=pelvis, loc=<2.9294 -3.7064 -14.7347>, rot=<-0.0109 -0.967 0.2249 -0.1192>}
 auto_skirt_midback_lf {parent=base_skirt_midback_lf, loc=<0 0 0>, rot=<0 0 0 1>}
 offset_skirt_midback_lf {parent=auto_skirt_midback_lf, loc=<0 0 0>, rot=<0 0 0 1>}
 dyn_skirt_midback_lf_01 {parent=offset_skirt_midback_lf, loc=<0 0 0>, rot=<0 0 0 1>}
 dyn_skirt_midback_lf_02 {parent=dyn_skirt_midback_lf_01, loc=<9.6195 0 0>, rot=<0.0333 0.0121 0.0373 -0.9987>}
 dyn_skirt_midback_lf_03 {parent=dyn_skirt_midback_lf_02, loc=<11.4431 0 0>, rot=<-0.0556 0.0328 0.0238 -0.9976>}
 dyn_skirt_midback_lf_04 {parent=dyn_skirt_midback_lf_03, loc=<9.3617 0 0>, rot=<0.0015 0.0116 0.0015 -0.9999>}
 base_skirt_back_lf {parent=pelvis, loc=<3.6445 -10.7206 -8.3059>, rot=<0.0243 -0.8242 0.5592 -0.0861>}
 auto_skirt_back_lf {parent=base_skirt_back_lf, loc=<0 0 0>, rot=<0 0 0 1>}
 offset_skirt_back_lf {parent=auto_skirt_back_lf, loc=<0 0 0>, rot=<0 0 0 1>}
 dyn_skirt_back_lf_01 {parent=offset_skirt_back_lf, loc=<0 0 0>, rot=<0 0 0 1>}
 dyn_skirt_back_lf_02 {parent=dyn_skirt_back_lf_01, loc=<10.5203 0 0>, rot=<0.0152 0.0187 -0.0076 -0.9997>}
 dyn_skirt_back_lf_03 {parent=dyn_skirt_back_lf_02, loc=<11.3627 0 0>, rot=<0.0336 0.0349 -0.0175 -0.9987>}
 dyn_skirt_back_lf_04 {parent=dyn_skirt_back_lf_03, loc=<9.2806 0 0>, rot=<0.0078 0.0043 -0.0036 -1>}
 base_skirt_back {parent=pelvis, loc=<3.8189 -12.3582 1.1947>, rot=<0.0595 -0.619 0.782 -0.0418>}
 auto_skirt_back {parent=base_skirt_back, loc=<0 0 0>, rot=<0 0 0 1>}
 offset_skirt_back {parent=auto_skirt_back, loc=<0 0 0>, rot=<0 0 0 1>}
 dyn_skirt_back_01 {parent=offset_skirt_back, loc=<0 0 0>, rot=<0 0 0 1>}
 dyn_skirt_back_02 {parent=dyn_skirt_back_01, loc=<10.2906 0 0>, rot=<0.0361 0.0225 -0.0318 -0.9986>}
 dyn_skirt_back_03 {parent=dyn_skirt_back_02, loc=<11.839 0 0>, rot=<0.0322 0.0085 -0.01 -0.9994>}
 dyn_skirt_back_04 {parent=dyn_skirt_back_03, loc=<9.3005 0 0>, rot=<-0.0399 0.0073 0.0092 -0.9991>}
 base_skirt_mid_rt {parent=pelvis, loc=<2.0101 2.8786 14.4666>, rot=<0.1574 0.0728 0.984 -0.0411>}
 auto_skirt_mid_rt {parent=base_skirt_mid_rt, loc=<0 0 0>, rot=<0 0 0 1>}
 offset_skirt_mid_rt {parent=auto_skirt_mid_rt, loc=<0 0 0>, rot=<0 0 0 1>}
 dyn_skirt_mid_rt_01 {parent=offset_skirt_mid_rt, loc=<0 0 0>, rot=<0 0 0 1>}
 dyn_skirt_mid_rt_02 {parent=dyn_skirt_mid_rt_01, loc=<9.6262 0 0>, rot=<0.0287 -0.0242 -0.0145 -0.9992>}
 dyn_skirt_mid_rt_03 {parent=dyn_skirt_mid_rt_02, loc=<11.7637 0 0>, rot=<0.0193 -0.0003 -0.007 -0.9998>}
 dyn_skirt_mid_rt_04 {parent=dyn_skirt_mid_rt_03, loc=<9.3706 0 0>, rot=<-0.0177 0.0251 -0.0074 -0.9995>}
 base_skirt_front_rt {parent=pelvis, loc=<1.5466 10.1151 8.6096>, rot=<0.1965 0.3456 0.915 -0.0687>}
 auto_skirt_front_rt {parent=base_skirt_front_rt, loc=<0 0 0>, rot=<0 0 0 1>}
 offset_skirt_front_rt {parent=auto_skirt_front_rt, loc=<0 0 0>, rot=<0 0 0 1>}
 dyn_skirt_front_rt_01 {parent=offset_skirt_front_rt, loc=<0 0 0>, rot=<0 0 0 1>}
 dyn_skirt_front_rt_02 {parent=dyn_skirt_front_rt_01, loc=<10.1882 0 0>, rot=<0.028 -0.0223 0.0056 -0.9993>}
 dyn_skirt_front_rt_03 {parent=dyn_skirt_front_rt_02, loc=<11.8072 0 0>, rot=<0.0431 -0.0238 0.0006 -0.9988>}
 dyn_skirt_front_rt_04 {parent=dyn_skirt_front_rt_03, loc=<9.53 0 0>, rot=<0.0234 -0.0087 -0.0031 -0.9997>}
 base_skirt_front_lf {parent=pelvis, loc=<1.7659 11.5539 -6.3847>, rot=<0.078 -0.8926 -0.396 -0.201>}
 auto_skirt_front_lf {parent=base_skirt_front_lf, loc=<0 0 0>, rot=<0 0 0 1>}
 offset_skirt_front_lf {parent=auto_skirt_front_lf, loc=<0 0 0>, rot=<0 0 0 1>}
 dyn_skirt_front_lf_01 {parent=offset_skirt_front_lf, loc=<0 0 0>, rot=<0 0 0 1>}
 dyn_skirt_front_lf_02 {parent=dyn_skirt_front_lf_01, loc=<9.4678 0 0>, rot=<-0.0548 -0.0191 0.0169 -0.9982>}
 dyn_skirt_front_lf_03 {parent=dyn_skirt_front_lf_02, loc=<11.5797 0 0>, rot=<-0.0424 -0.0145 0.0135 -0.9989>}
 dyn_skirt_front_lf_04 {parent=dyn_skirt_front_lf_03, loc=<10.0142 0 0>, rot=<-0.0115 -0.0083 -0.0016 -0.9999>}
 base_bag_front_rt {parent=pelvis, loc=<10.2258 4.0703 12.4924>, rot=<0.1383 0.2563 0.9561 0.0315>}
 dyn_bag_front_rt {parent=base_bag_front_rt, loc=<0 0 0>, rot=<0 0 0 1>}
 base_bag_back_rt {parent=pelvis, loc=<10.7826 -8.8124 12.1044>, rot=<0.0287 -0.3664 0.9293 -0.0366>}
 dyn_bag_back_rt {parent=base_bag_back_rt, loc=<0 0 0>, rot=<0 0 0 1>}
 base_bag_front_lf {parent=pelvis, loc=<10.2534 6.8965 -12.9272>, rot=<0.0107 -0.9571 -0.2664 -0.1136>}
 dyn_bag_front_lf {parent=base_bag_front_lf, loc=<0 0 0>, rot=<0 0 0 1>}
 base_bag_back_lf {parent=pelvis, loc=<10.8075 -4.6941 -14.9626>, rot=<0.0499 -0.9801 0.1795 -0.0683>}
 dyn_bag_back_lf {parent=base_bag_back_lf, loc=<0 0 0>, rot=<0 0 0 1>}
 thigh_r_ctrl {controlType=EULER_TRANSFORM, type=Control, loc=<0 0 0>, rot=<0 0 0 1>}
 calf_r_ctrl {controlType=EULER_TRANSFORM, parent=thigh_r_ctrl, type=Control, loc=<0 0 0>, rot=<0 0 0 1>}
 thigh_l_ctrl {controlType=EULER_TRANSFORM, type=Control, loc=<0 0 0>, rot=<0 0 0 1>}
 calf_l_ctrl {controlType=EULER_TRANSFORM, parent=thigh_l_ctrl, type=Control, loc=<0 0 0>, rot=<0 0 0 1>}
 upperarm_r_ctrl {controlType=EULER_TRANSFORM, type=Control, loc=<0 0 0>, rot=<0 0 0 1>}
 lowerarm_r_ctrl {controlType=EULER_TRANSFORM, parent=upperarm_r_ctrl, type=Control, loc=<0 0 0>, rot=<0 0 0 1>}
 hand_r_ctrl {controlType=EULER_TRANSFORM, parent=lowerarm_r_ctrl, type=Control, loc=<0 0 0>, rot=<0 0 0 1>}
 clavicle_l_ctrl {controlType=EULER_TRANSFORM, type=Control, loc=<0 0 0>, rot=<0 0 0 1>}
 upperarm_l_ctrl {controlType=EULER_TRANSFORM, parent=clavicle_l_ctrl, type=Control, loc=<0 0 0>, rot=<0 0 0 1>}
 lowerarm_l_ctrl {controlType=EULER_TRANSFORM, parent=upperarm_l_ctrl, type=Control, loc=<0 0 0>, rot=<0 0 0 1>}
 hand_l_ctrl {controlType=EULER_TRANSFORM, parent=lowerarm_l_ctrl, type=Control, loc=<0 0 0>, rot=<0 0 0 1>}
]
variables [
 CrouchingButUp {type=double, default=0}
 CrouchingNotMove {type=bool, default=False}
 OnAir {type=bool, default=False}
 Sliding {type=bool, default=False}
 Sprint {type=bool, default=False}
]
