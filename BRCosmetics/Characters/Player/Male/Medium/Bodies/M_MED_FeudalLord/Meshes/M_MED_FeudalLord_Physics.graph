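# Physics asset for the Feudal Lord medium male player body cosmetic, covering its dangling skirt and four bags plus the lower body.
# Nine four-bone skirt chains hang from tiny offset anchors on the pelvis, and each bag is one capsule on a base anchor.
# Skirt swing limits generally widen toward the hem, with twist nearly locked.
# Pelvis, thigh, calf and foot capsules form the body volume, linked by free angular constraints like the anchors.
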PhysicsAsset /BRCosmetics/Characters/Player/Male/Medium/Bodies/M_MED_FeudalLord/Meshes/M_MED_FeudalLord_Physics
bodies {
 base_bag_back_lf [{shapeType=Capsule, length=0.505, radius=0.505, shapeName=base_bag_back_lf_capsule, rotation=<rot 0 -90 0>}]
 base_bag_back_rt [{shapeType=Capsule, length=0.505, radius=0.505, shapeName=base_bag_back_rt_capsule, rotation=<rot 0 -90 0>}]
 base_bag_front_lf [{shapeType=Capsule, length=0.505, radius=0.505, shapeName=base_bag_front_lf_capsule, rotation=<rot 0 -90 0>}]
 base_bag_front_rt [{shapeType=Capsule, length=0.505, radius=0.505, shapeName=base_bag_front_rt_capsule, rotation=<rot 0 -90 0>}]
 calf_l [{shapeType=Capsule, length=41, radius=11, shapeName=calf_l_capsule, center=<-24 -1 0>, rotation=<rot 90 0 -5>}]
 calf_r [{shapeType=Capsule, length=41, radius=11, shapeName=calf_r_capsule, center=<24 1 0>, rotation=<rot 90 180 175>}]
 dyn_bag_back_lf [{shapeType=Capsule, length=7, radius=4, shapeName=dyn_bag_back_lf_capsule, center=<10.5 0 10>, rotation=<rot 0 0 90>}]
 dyn_bag_back_rt [{shapeType=Capsule, length=7, radius=4, shapeName=dyn_bag_back_rt_capsule, center=<8.4 0 10>, rotation=<rot 0 0 90>}]
 dyn_bag_front_lf [{shapeType=Capsule, length=11, radius=4, shapeName=dyn_bag_front_lf_capsule, center=<12 0 10.2>, rotation=<rot 0 0 90>}]
 dyn_bag_front_rt [{shapeType=Capsule, length=4, radius=4, shapeName=dyn_bag_front_rt_capsule, center=<8 0 9.5>, rotation=<rot 0 0 90>}]
 dyn_skirt_back_01 [{shapeType=Capsule, length=11, radius=5.5, shapeName=dyn_skirt_back_01_capsule, center=<5 0 2>, rotation=<rot 90 0 0>}]
 dyn_skirt_back_02 [{shapeType=Capsule, length=11, radius=6.1, shapeName=dyn_skirt_back_02_capsule, center=<5 0 2.5>, rotation=<rot 0 0 90>}]
 dyn_skirt_back_03 [{shapeType=Capsule, length=14, radius=7.5, shapeName=dyn_skirt_back_03_capsule, center=<5 0 2.9>, rotation=<rot 0 0 90>}]
 dyn_skirt_back_04 [{shapeType=Capsule, length=16, radius=7.5, shapeName=dyn_skirt_back_04_capsule, center=<6.6574 0 2.9>, rotation=<rot 0 0 90>}]
 dyn_skirt_back_lf_01 [{shapeType=Capsule, length=11, radius=5.5, shapeName=dyn_skirt_back_lf_01_capsule, center=<5 0 2>, rotation=<rot 90 0 0>}]
 dyn_skirt_back_lf_02 [{shapeType=Capsule, length=11, radius=6.1, shapeName=dyn_skirt_back_lf_02_capsule, center=<5.5 0 2>, rotation=<rot 90 -10.025 -10.025>}]
 dyn_skirt_back_lf_03 [{shapeType=Capsule, length=16, radius=7.5, shapeName=dyn_skirt_back_lf_03_capsule, center=<5 0 2.9>, rotation=<rot 0 0 90>}]
 dyn_skirt_back_lf_04 [{shapeType=Capsule, length=22, radius=7.5, shapeName=dyn_skirt_back_lf_04_capsule, center=<6.6574 0 2.9>, rotation=<rot 0 0 90>}]
 dyn_skirt_back_rt_01 [{shapeType=Capsule, length=11, radius=5.5, shapeName=dyn_skirt_back_rt_01_capsule, center=<5 0 2>, rotation=<rot 90 0 0>}]
 dyn_skirt_back_rt_02 [{shapeType=Capsule, length=11, radius=6.1, shapeName=dyn_skirt_back_rt_02_capsule, center=<5.5 0 2>, rotation=<rot 90 0 0>}]
 dyn_skirt_back_rt_03 [{shapeType=Capsule, length=16, radius=7.5, shapeName=dyn_skirt_back_rt_03_capsule, center=<5 0 2.9>, rotation=<rot 0 0 90>}]
 dyn_skirt_back_rt_04 [{shapeType=Capsule, length=22, radius=7.5, shapeName=dyn_skirt_back_rt_04_capsule, center=<6.6574 0 2.9>, rotation=<rot 0 0 90>}]
 dyn_skirt_front_lf_01 [{shapeType=Capsule, length=11, radius=5, shapeName=dyn_skirt_front_lf_01_capsule, center=<8 2.5 2.5>, rotation=<rot 90 0 0>}]
 dyn_skirt_front_lf_02 [{shapeType=Capsule, length=18, radius=7, shapeName=dyn_skirt_front_lf_02_capsule, center=<5 2.5 2.2>, rotation=<rot 90 0 0>}]
 dyn_skirt_front_lf_03 [{shapeType=Capsule, length=18, radius=7.3, shapeName=dyn_skirt_front_lf_03_capsule, center=<6.055 0 3.3>, rotation=<rot 0 0 90>}]
 dyn_skirt_front_lf_04 [{shapeType=Capsule, length=22, radius=6.5, shapeName=dyn_skirt_front_lf_04_capsule, center=<7.6104 0 3.3>, rotation=<rot 0 0 90>}]
 dyn_skirt_front_rt_01 [{shapeType=Capsule, length=7.8, radius=5, shapeName=dyn_skirt_front_rt_01_capsule, center=<5 0 2.5>, rotation=<rot 90 0 0>}]
 dyn_skirt_front_rt_02 [{shapeType=Capsule, length=9, radius=6, shapeName=dyn_skirt_front_rt_02_capsule, center=<5 -1 2.5>, rotation=<rot 90 0 0>}]
 dyn_skirt_front_rt_03 [{shapeType=Capsule, length=18, radius=6.5, shapeName=dyn_skirt_front_rt_03_capsule, center=<6.055 0 3.3>, rotation=<rot 0 0 90>}]
 dyn_skirt_front_rt_04 [{shapeType=Capsule, length=22, radius=6, shapeName=dyn_skirt_front_rt_04_capsule, center=<7.6104 0 3.3>, rotation=<rot 0 0 90>}]
 dyn_skirt_mid_lf_01 [{shapeType=Capsule, length=5, radius=5, shapeName=dyn_skirt_mid_lf_01_capsule, center=<5 0 2>, rotation=<rot 90 0 0>}]
 dyn_skirt_mid_lf_02 [{shapeType=Capsule, length=9, radius=6, shapeName=dyn_skirt_mid_lf_02_capsule, center=<5 0 2>, rotation=<rot 90 0 0>}]
 dyn_skirt_mid_lf_03 [{shapeType=Capsule, length=22, radius=6.5, shapeName=dyn_skirt_mid_lf_03_capsule, center=<6.055 0 2.7>, rotation=<rot 0 0 90>}]
 dyn_skirt_mid_lf_04 [{shapeType=Capsule, length=26, radius=7, shapeName=dyn_skirt_mid_lf_04_capsule, center=<7.6104 0 2.7>, rotation=<rot 0 0 90>}]
 dyn_skirt_mid_rt_01 [{shapeType=Capsule, length=5, radius=5, shapeName=dyn_skirt_mid_rt_01_capsule, center=<5 0 2>, rotation=<rot 90 0 0>}]
 dyn_skirt_mid_rt_02 [{shapeType=Capsule, length=9, radius=6, shapeName=dyn_skirt_mid_rt_02_capsule, center=<5 0 2>, rotation=<rot 90 0 0>}]
 dyn_skirt_mid_rt_03 [{shapeType=Capsule, length=22, radius=6.5, shapeName=dyn_skirt_mid_rt_03_capsule, center=<6.055 0 2.7>, rotation=<rot 0 0 90>}]
 dyn_skirt_mid_rt_04 [{shapeType=Capsule, length=26, radius=7, shapeName=dyn_skirt_mid_rt_04_capsule, center=<7.6104 0 2.7>, rotation=<rot 0 0 90>}]
 dyn_skirt_midback_lf_01 [{shapeType=Capsule, length=3.1288, radius=5, shapeName=dyn_skirt_midback_lf_01_capsule, center=<5 0 2>, rotation=<rot 90 0 0>}]
 dyn_skirt_midback_lf_02 [{shapeType=Capsule, length=6, radius=6, shapeName=dyn_skirt_midback_lf_02_capsule, center=<5 0 2>, rotation=<rot 90 -30.0319 -30.0319>}]
 dyn_skirt_midback_lf_03 [{shapeType=Capsule, length=18, radius=7, shapeName=dyn_skirt_midback_lf_03_capsule, center=<5 0 2.5>, rotation=<rot 0 0 90>}]
 dyn_skirt_midback_lf_04 [{shapeType=Capsule, length=22, radius=7, shapeName=dyn_skirt_midback_lf_04_capsule, center=<6.6574 0 2.6>, rotation=<rot 0 0 90>}]
 dyn_skirt_midback_rt_01 [{shapeType=Capsule, length=2.8006, radius=5, shapeName=dyn_skirt_midback_rt_01_capsule, center=<5 0 2>, rotation=<rot 90 0 0>}]
 dyn_skirt_midback_rt_02 [{shapeType=Capsule, length=6, radius=6, shapeName=dyn_skirt_midback_rt_02_capsule, center=<5 0 2>, rotation=<rot 90 0 0>}]
 dyn_skirt_midback_rt_03 [{shapeType=Capsule, length=18, radius=7, shapeName=dyn_skirt_midback_rt_03_capsule, center=<5 0 2.5>, rotation=<rot 0 0 90>}]
 dyn_skirt_midback_rt_04 [{shapeType=Capsule, length=22, radius=7, shapeName=dyn_skirt_midback_rt_04_capsule, center=<6.6574 0 2.6>, rotation=<rot 0 0 90>}]
 foot_l [{shapeType=Capsule, length=22, radius=11, shapeName=foot_l_capsule, center=<-4.5 -3 3>}]
 foot_r [{shapeType=Capsule, length=18, radius=11, shapeName=foot_r_capsule, center=<4.5 3 -6>}]
 offset_skirt_back [{shapeType=Capsule, length=0.505, radius=0.505, shapeName=offset_skirt_back_capsule, rotation=<rot 0 -90 0>}]
 offset_skirt_back_lf [{shapeType=Capsule, length=0.505, radius=0.505, shapeName=offset_skirt_back_lf_capsule, rotation=<rot 0 -90 0>}]
 offset_skirt_back_rt [{shapeType=Capsule, length=0.505, radius=0.505, shapeName=offset_skirt_back_rt_capsule, rotation=<rot 0 -90 0>}]
 offset_skirt_front_lf [{shapeType=Capsule, length=0.505, radius=0.505, shapeName=offset_skirt_front_lf_capsule, rotation=<rot 0 -90 0>}]
 offset_skirt_front_rt [{shapeType=Capsule, length=0.505, radius=0.505, shapeName=offset_skirt_front_rt_capsule, rotation=<rot 0 -90 0>}]
 offset_skirt_mid_lf [{shapeType=Capsule, length=0.505, radius=0.505, shapeName=offset_skirt_mid_lf_capsule, rotation=<rot 0 -90 0>}]
 offset_skirt_mid_rt [{shapeType=Capsule, length=0.505, radius=0.505, shapeName=offset_skirt_mid_rt_capsule, rotation=<rot 0 -90 0>}]
 offset_skirt_midback_lf [{shapeType=Capsule, length=0.505, radius=0.505, shapeName=offset_skirt_midback_lf_capsule, rotation=<rot 0 -90 0>}]
 offset_skirt_midback_rt [{shapeType=Capsule, length=0.505, radius=0.505, shapeName=offset_skirt_midback_rt_capsule, rotation=<rot 0 -90 0>}]
 pelvis [{shapeType=Capsule, length=16, radius=9, shapeName=pelvis_capsule}]
 thigh_l [{shapeType=Capsule, length=33, radius=9, shapeName=thigh_l_capsule, center=<-22.94 1.1 1.1115>, rotation=<rot 85 180 180>}]
 thigh_r [{shapeType=Capsule, length=22, radius=9, shapeName=thigh_r_capsule, center=<18.5 -1 -1.5>, rotation=<rot 85 180 180>}]
}
constraints [
 {bone1Name=dyn_bag_back_lf, bone2Name=base_bag_back_lf, swing1LimitDegrees=10, swing2LimitDegrees=25, twistLimitDegrees=1}
 {bone1Name=dyn_bag_back_rt, bone2Name=base_bag_back_rt, swing1LimitDegrees=10, swing2LimitDegrees=25, twistLimitDegrees=1}
 {bone1Name=dyn_bag_front_lf, bone2Name=base_bag_front_lf, swing1LimitDegrees=10, swing2LimitDegrees=25, twistLimitDegrees=1}
 {bone1Name=dyn_bag_front_rt, bone2Name=base_bag_front_rt, swing1LimitDegrees=10, swing2LimitDegrees=25, twistLimitDegrees=1}
 {bone1Name=dyn_skirt_back_02, bone2Name=dyn_skirt_back_01, swing1LimitDegrees=5, swing2LimitDegrees=20, twistLimitDegrees=1}
 {bone1Name=dyn_skirt_back_03, bone2Name=dyn_skirt_back_02, swing1LimitDegrees=8, swing2LimitDegrees=25, twistLimitDegrees=1}
 {bone1Name=dyn_skirt_back_04, bone2Name=dyn_skirt_back_03, swing1LimitDegrees=10, swing2LimitDegrees=30, twistLimitDegrees=1}
 {bone1Name=dyn_skirt_back_lf_02, bone2Name=dyn_skirt_back_lf_01, swing1LimitDegrees=5, swing2LimitDegrees=20, twistLimitDegrees=1}
 {bone1Name=dyn_skirt_back_lf_03, bone2Name=dyn_skirt_back_lf_02, swing1LimitDegrees=8, swing2LimitDegrees=25, twistLimitDegrees=1}
 {bone1Name=dyn_skirt_back_lf_04, bone2Name=dyn_skirt_back_lf_03, swing1LimitDegrees=10, swing2LimitDegrees=30, twistLimitDegrees=1}
 {bone1Name=dyn_skirt_back_rt_02, bone2Name=dyn_skirt_back_rt_01, swing1LimitDegrees=5, swing2LimitDegrees=20, twistLimitDegrees=1}
 {bone1Name=dyn_skirt_back_rt_03, bone2Name=dyn_skirt_back_rt_02, swing1LimitDegrees=8, swing2LimitDegrees=25, twistLimitDegrees=1}
 {bone1Name=dyn_skirt_back_rt_04, bone2Name=dyn_skirt_back_rt_03, swing1LimitDegrees=10, swing2LimitDegrees=30, twistLimitDegrees=1}
 {bone1Name=dyn_skirt_front_lf_02, bone2Name=dyn_skirt_front_lf_01, swing1LimitDegrees=5, swing2LimitDegrees=20, twistLimitDegrees=1}
 {bone1Name=dyn_skirt_front_lf_03, bone2Name=dyn_skirt_front_lf_02, swing1LimitDegrees=5, swing2LimitDegrees=25, twistLimitDegrees=1}
 {bone1Name=dyn_skirt_front_lf_04, bone2Name=dyn_skirt_front_lf_03, swing1LimitDegrees=5, swing2LimitDegrees=30, twistLimitDegrees=1}
 {bone1Name=dyn_skirt_front_rt_02, bone2Name=dyn_skirt_front_rt_01, swing1LimitDegrees=5, swing2LimitDegrees=20, twistLimitDegrees=1}
 {bone1Name=dyn_skirt_front_rt_03, bone2Name=dyn_skirt_front_rt_02, swing1LimitDegrees=8, swing2LimitDegrees=25, twistLimitDegrees=1}
 {bone1Name=dyn_skirt_front_rt_04, bone2Name=dyn_skirt_front_rt_03, swing1LimitDegrees=10, swing2LimitDegrees=30, twistLimitDegrees=1}
 {bone1Name=dyn_skirt_mid_lf_02, bone2Name=dyn_skirt_mid_lf_01, swing1LimitDegrees=5, swing2LimitDegrees=20, twistLimitDegrees=1}
 {bone1Name=dyn_skirt_mid_lf_03, bone2Name=dyn_skirt_mid_lf_02, swing1LimitDegrees=8, swing2LimitDegrees=25, twistLimitDegrees=1}
 {bone1Name=dyn_skirt_mid_lf_04, bone2Name=dyn_skirt_mid_lf_03, swing1LimitDegrees=10, swing2LimitDegrees=30, twistLimitDegrees=1}
 {bone1Name=dyn_skirt_mid_rt_02, bone2Name=dyn_skirt_mid_rt_01, swing1LimitDegrees=5, swing2LimitDegrees=20, twistLimitDegrees=1}
 {bone1Name=dyn_skirt_mid_rt_03, bone2Name=dyn_skirt_mid_rt_02, swing1LimitDegrees=8, swing2LimitDegrees=25, twistLimitDegrees=1}
 {bone1Name=dyn_skirt_mid_rt_04, bone2Name=dyn_skirt_mid_rt_03, swing1LimitDegrees=10, swing2LimitDegrees=30, twistLimitDegrees=1}
 {bone1Name=dyn_skirt_midback_lf_02, bone2Name=dyn_skirt_midback_lf_01, swing1LimitDegrees=5, swing2LimitDegrees=20, twistLimitDegrees=1}
 {bone1Name=dyn_skirt_midback_lf_03, bone2Name=dyn_skirt_midback_lf_02, swing1LimitDegrees=8, swing2LimitDegrees=25, twistLimitDegrees=1}
 {bone1Name=dyn_skirt_midback_lf_04, bone2Name=dyn_skirt_midback_lf_03, swing1LimitDegrees=10, swing2LimitDegrees=30, twistLimitDegrees=1}
 {bone1Name=dyn_skirt_midback_rt_02, bone2Name=dyn_skirt_midback_rt_01, swing1LimitDegrees=5, swing2LimitDegrees=20, twistLimitDegrees=1}
 {bone1Name=dyn_skirt_midback_rt_03, bone2Name=dyn_skirt_midback_rt_02, swing1LimitDegrees=8, swing2LimitDegrees=25, twistLimitDegrees=1}
 {bone1Name=dyn_skirt_midback_rt_04, bone2Name=dyn_skirt_midback_rt_03, swing1LimitDegrees=10, swing2LimitDegrees=30, twistLimitDegrees=1}
 {bone1Name=dyn_skirt_back_01, bone2Name=offset_skirt_back, swing1LimitDegrees=3, swing2LimitDegrees=15, twistLimitDegrees=1}
 {bone1Name=dyn_skirt_back_lf_01, bone2Name=offset_skirt_back_lf, swing1LimitDegrees=3, swing2LimitDegrees=15, twistLimitDegrees=1}
 {bone1Name=dyn_skirt_back_rt_01, bone2Name=offset_skirt_back_rt, swing1LimitDegrees=3, swing2LimitDegrees=15, twistLimitDegrees=1}
 {bone1Name=dyn_skirt_front_lf_01, bone2Name=offset_skirt_front_lf, swing1LimitDegrees=3, swing2LimitDegrees=15, twistLimitDegrees=1}
 {bone1Name=dyn_skirt_front_rt_01, bone2Name=offset_skirt_front_rt, swing1LimitDegrees=3, swing2LimitDegrees=15, twistLimitDegrees=1}
 {bone1Name=dyn_skirt_mid_lf_01, bone2Name=offset_skirt_mid_lf, swing1LimitDegrees=3, swing2LimitDegrees=15, twistLimitDegrees=1}
 {bone1Name=dyn_skirt_mid_rt_01, bone2Name=offset_skirt_mid_rt, swing1LimitDegrees=3, swing2LimitDegrees=15, twistLimitDegrees=1}
 {bone1Name=dyn_skirt_midback_lf_01, bone2Name=offset_skirt_midback_lf, swing1LimitDegrees=3, swing2LimitDegrees=15, twistLimitDegrees=1}
 {bone1Name=dyn_skirt_midback_rt_01, bone2Name=offset_skirt_midback_rt, swing1LimitDegrees=3, swing2LimitDegrees=15, twistLimitDegrees=1}
 {
  bone1Name=foot_l
  bone2Name=calf_l
  swing1LimitDegrees=45
  swing1Motion=Free
  swing2LimitDegrees=45
  swing2Motion=Free
  twistLimitDegrees=45
  twistMotion=Free
 }
 {
  bone1Name=calf_l
  bone2Name=thigh_l
  swing1LimitDegrees=45
  swing1Motion=Free
  swing2LimitDegrees=45
  swing2Motion=Free
  twistLimitDegrees=45
  twistMotion=Free
 }
 {
  bone1Name=calf_r
  bone2Name=thigh_r
  swing1LimitDegrees=45
  swing1Motion=Free
  swing2LimitDegrees=45
  swing2Motion=Free
  twistLimitDegrees=45
  twistMotion=Free
 }
 {
  bone1Name=foot_r
  bone2Name=calf_r
  swing1LimitDegrees=45
  swing1Motion=Free
  swing2LimitDegrees=45
  swing2Motion=Free
  twistLimitDegrees=45
  twistMotion=Free
 }
 {
  bone1Name=thigh_l
  bone2Name=pelvis
  swing1LimitDegrees=45
  swing1Motion=Free
  swing2LimitDegrees=45
  swing2Motion=Free
  twistLimitDegrees=45
  twistMotion=Free
 }
 {
  bone1Name=thigh_r
  bone2Name=pelvis
  swing1LimitDegrees=45
  swing1Motion=Free
  swing2LimitDegrees=45
  swing2Motion=Free
  twistLimitDegrees=45
  twistMotion=Free
 }
 {
  bone1Name=offset_skirt_back_rt
  bone2Name=pelvis
  swing1LimitDegrees=45
  swing1Motion=Free
  swing2LimitDegrees=45
  swing2Motion=Free
  twistLimitDegrees=45
  twistMotion=Free
 }
 {
  bone1Name=offset_skirt_mid_lf
  bone2Name=pelvis
  swing1LimitDegrees=45
  swing1Motion=Free
  swing2LimitDegrees=45
  swing2Motion=Free
  twistLimitDegrees=45
  twistMotion=Free
 }
 {
  bone1Name=offset_skirt_midback_rt
  bone2Name=pelvis
  swing1LimitDegrees=45
  swing1Motion=Free
  swing2LimitDegrees=45
  swing2Motion=Free
  twistLimitDegrees=45
  twistMotion=Free
 }
 {
  bone1Name=offset_skirt_midback_lf
  bone2Name=pelvis
  swing1LimitDegrees=45
  swing1Motion=Free
  swing2LimitDegrees=45
  swing2Motion=Free
  twistLimitDegrees=45
  twistMotion=Free
 }
 {
  bone1Name=offset_skirt_back_lf
  bone2Name=pelvis
  swing1LimitDegrees=45
  swing1Motion=Free
  swing2LimitDegrees=45
  swing2Motion=Free
  twistLimitDegrees=45
  twistMotion=Free
 }
 {
  bone1Name=offset_skirt_back
  bone2Name=pelvis
  swing1LimitDegrees=45
  swing1Motion=Free
  swing2LimitDegrees=45
  swing2Motion=Free
  twistLimitDegrees=45
  twistMotion=Free
 }
 {
  bone1Name=offset_skirt_mid_rt
  bone2Name=pelvis
  swing1LimitDegrees=45
  swing1Motion=Free
  swing2LimitDegrees=45
  swing2Motion=Free
  twistLimitDegrees=45
  twistMotion=Free
 }
 {
  bone1Name=offset_skirt_front_rt
  bone2Name=pelvis
  swing1LimitDegrees=45
  swing1Motion=Free
  swing2LimitDegrees=45
  swing2Motion=Free
  twistLimitDegrees=45
  twistMotion=Free
 }
 {
  bone1Name=offset_skirt_front_lf
  bone2Name=pelvis
  swing1LimitDegrees=45
  swing1Motion=Free
  swing2LimitDegrees=45
  swing2Motion=Free
  twistLimitDegrees=45
  twistMotion=Free
 }
 {
  bone1Name=base_bag_front_rt
  bone2Name=pelvis
  swing1LimitDegrees=45
  swing1Motion=Free
  swing2LimitDegrees=45
  swing2Motion=Free
  twistLimitDegrees=45
  twistMotion=Free
 }
 {
  bone1Name=base_bag_back_rt
  bone2Name=pelvis
  swing1LimitDegrees=45
  swing1Motion=Free
  swing2LimitDegrees=45
  swing2Motion=Free
  twistLimitDegrees=45
  twistMotion=Free
 }
 {
  bone1Name=base_bag_front_lf
  bone2Name=pelvis
  swing1LimitDegrees=45
  swing1Motion=Free
  swing2LimitDegrees=45
  swing2Motion=Free
  twistLimitDegrees=45
  twistMotion=Free
 }
 {
  bone1Name=base_bag_back_lf
  bone2Name=pelvis
  swing1LimitDegrees=45
  swing1Motion=Free
  swing2LimitDegrees=45
  swing2Motion=Free
  twistLimitDegrees=45
  twistMotion=Free
 }
]
