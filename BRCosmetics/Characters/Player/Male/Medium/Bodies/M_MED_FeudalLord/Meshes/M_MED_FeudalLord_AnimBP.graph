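# Animation Blueprint for the Feudal Lord male cosmetic body, adding secondary motion to its skirt and bags.
# The parent mesh's copied pose feeds a state-driven Control Rig, pelvis-space RigidBody simulation, then stacked additive skirt and bag bone offsets.
# Each update derives gravity and sim-space settings from per-movement-state RBAN parameters via a deprecated helper.
# It also sets crouch, LOD and knee-versus-baseball slide values, the latter from left calf rotation.
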AnimBlueprint /BRCosmetics/Characters/Player/Male/Medium/Bodies/M_MED_FeudalLord/Meshes/M_MED_FeudalLord_AnimBP
target_skeleton=&/M_MED_FeudalLord_Skeleton
graphs [
 AnimGraph {nodes [
   n1 = Control Rig!(
    Source=Copy Pose From Mesh() {bUseAttachedParent=true}
    Alpha
    Sliding:bool=Get bIsSliding()
    Sprint:bool=Get bIsTacticalSprinting()
    CrouchingNotMove:bool=Get CrouchingNotMove() -> (CrouchingNotMove:bool)
    OnAir:bool=Get bIsSkyDiving()
   ) -> (Pose) {
    alphaBoolBlend {blendOption=Linear}
    controlRigAssetReference {blueprintRigClass=&/M_MED_FeudalLord_CtrlRig.M_MED_FeudalLord_CtrlRig_C}
   }
   n2 = RigidBody!(
    ComponentPose=Local To Component(LocalPose=@n1)
    Alpha
    OverrideWorldGravity:Vector=Get GravityOverride()
    ExternalForce
    SimSpaceSettings:Sim Space Settings Structure=Get Sim Space Settings() -> (Sim Space Settings:Sim Space Settings Structure)
   ) -> (Pose) {
    bClampLinearTranslationLimitToRefPose=true
    bOverrideWorldGravity=true
    overridePhysicsAsset=&/M_MED_FeudalLord_Physics
    simulationSpace=BaseBoneSpace
    alphaBoolBlend {blendOption=Linear}
    baseBoneRef {boneName=pelvis}
    simSpaceSettings {maxAngularAcceleration=1, maxAngularVelocity=3, maxLinearAcceleration=1000, maxLinearVelocity=500, worldAlpha=1, externalLinearDragV=<1 1 1>}
   }
   n3 = "Transform (Modify) Bone - Bone: base_skirt_front_lf"!(ComponentPose=@n2, bAlphaBoolEnabled:bool=True, Translation="-2,0,0", Rotation="-15.5,0,0", Scale="1,1,1") -> (Pose) {
    alphaInputType=Bool
    rotationMode=BMM_Additive
    rotationSpace=BCS_BoneSpace
    translationMode=BMM_Additive
    translationSpace=BCS_BoneSpace
    alphaBoolBlend {blendInTime=0.5, blendOption=Sinusoidal, blendOutTime=1}
    boneToModify {boneName=base_skirt_front_lf}
    rotation {pitch=-15.5}
    translation {x=-2}
   }
   n4 = "Transform (Modify) Bone - Bone: base_skirt_front_lf"(ComponentPose=@n3, bAlphaBoolEnabled=True, Translation="0,0,0", Rotation="10,0,0", Scale="1,1,1") {
    alphaInputType=Bool
    rotationMode=BMM_Additive
    rotationSpace=BCS_BoneSpace
    translationMode=BMM_Additive
    translationSpace=BCS_BoneSpace
    alphaBoolBlend {blendInTime=0.5, blendOption=Sinusoidal, blendOutTime=1}
    boneToModify {boneName=base_skirt_front_lf}
    rotation {pitch=10}
   }
   n5 = "Transform (Modify) Bone - Bone: base_skirt_mid_lf"!(ComponentPose=@n4, bAlphaBoolEnabled:bool=True, Translation="0,0,0", Rotation="5,0,0", Scale="1,1,1") -> (Pose) {
    alphaInputType=Bool
    rotationMode=BMM_Additive
    rotationSpace=BCS_BoneSpace
    translationMode=BMM_Additive
    translationSpace=BCS_BoneSpace
    alphaBoolBlend {blendInTime=0.5, blendOption=Sinusoidal, blendOutTime=1}
    boneToModify {boneName=base_skirt_mid_lf}
    rotation {pitch=5}
   }
   n6 = "Transform (Modify) Bone - Bone: base_skirt_mid_rt"!(ComponentPose=@n5, bAlphaBoolEnabled:bool=True, Translation="0,0,0", Rotation="5,0,0", Scale="1,1,1") -> (Pose) {
    alphaInputType=Bool
    rotationMode=BMM_Additive
    rotationSpace=BCS_BoneSpace
    translationMode=BMM_Additive
    translationSpace=BCS_BoneSpace
    alphaBoolBlend {blendInTime=0.5, blendOption=Sinusoidal, blendOutTime=1}
    boneToModify {boneName=base_skirt_mid_rt}
    rotation {pitch=5}
   }
   n7 = "Transform (Modify) Bone - Bone: base_skirt_front_rt"!(ComponentPose=@n6, bAlphaBoolEnabled:bool=True, Translation="0,0,0", Rotation="10,0,0", Scale="1,1,1") -> (Pose) {
    alphaInputType=Bool
    rotationMode=BMM_Additive
    rotationSpace=BCS_BoneSpace
    translationMode=BMM_Additive
    translationSpace=BCS_BoneSpace
    alphaBoolBlend {blendInTime=0.5, blendOption=Sinusoidal, blendOutTime=1}
    boneToModify {boneName=base_skirt_front_rt}
    rotation {pitch=10}
   }
   n8 = "Transform (Modify) Bone - Bone: base_skirt_midback_lf"!(ComponentPose=@n7, bAlphaBoolEnabled:bool=True, Translation="0,0,0", Rotation="20,0,0", Scale="1,1,1") -> (Pose) {
    alphaInputType=Bool
    rotationMode=BMM_Additive
    rotationSpace=BCS_BoneSpace
    translationMode=BMM_Additive
    translationSpace=BCS_BoneSpace
    alphaBoolBlend {blendInTime=0.5, blendOption=Sinusoidal, blendOutTime=1}
    boneToModify {boneName=base_skirt_midback_lf}
    rotation {pitch=20}
   }
   n9 = "Transform (Modify) Bone - Bone: base_skirt_back_lf"!(ComponentPose=@n8, bAlphaBoolEnabled:bool=True, Translation="0,0,0", Rotation="25,0,0", Scale="1,1,1") -> (Pose) {
    alphaInputType=Bool
    rotationMode=BMM_Additive
    rotationSpace=BCS_BoneSpace
    translationMode=BMM_Additive
    translationSpace=BCS_BoneSpace
    alphaBoolBlend {blendInTime=0.5, blendOption=Sinusoidal, blendOutTime=1}
    boneToModify {boneName=base_skirt_back_lf}
    rotation {pitch=25}
   }
   n10 = "Transform (Modify) Bone - Bone: base_skirt_back"!(ComponentPose=@n9, bAlphaBoolEnabled:bool=True, Translation="0,0,0", Rotation="25,0,0", Scale="1,1,1") -> (Pose) {
    alphaInputType=Bool
    rotationMode=BMM_Additive
    rotationSpace=BCS_BoneSpace
    translationMode=BMM_Additive
    translationSpace=BCS_BoneSpace
    alphaBoolBlend {blendInTime=0.5, blendOption=Sinusoidal, blendOutTime=1}
    boneToModify {boneName=base_skirt_back}
    rotation {pitch=25}
   }
   n11 = "Transform (Modify) Bone - Bone: base_skirt_back_rt"!(ComponentPose=@n10, bAlphaBoolEnabled:bool=True, Translation="0,0,0", Rotation="25,0,0", Scale="1,1,1") -> (Pose) {
    alphaInputType=Bool
    rotationMode=BMM_Additive
    rotationSpace=BCS_BoneSpace
    translationMode=BMM_Additive
    translationSpace=BCS_BoneSpace
    alphaBoolBlend {blendInTime=0.5, blendOption=Sinusoidal, blendOutTime=1}
    boneToModify {boneName=base_skirt_back_rt}
    rotation {pitch=25}
   }
   n12 = "Transform (Modify) Bone - Bone: base_skirt_midback_lf"!(ComponentPose=@n11, Alpha:float=1, Translation="0,0,0", Rotation="30,0,0", Scale="1,1,1") -> (Pose) {
    rotationMode=BMM_Additive
    rotationSpace=BCS_BoneSpace
    translationMode=BMM_Additive
    translationSpace=BCS_BoneSpace
    alphaBoolBlend {blendInTime=0.5, blendOption=Sinusoidal, blendOutTime=1}
    boneToModify {boneName=base_skirt_midback_lf}
    rotation {pitch=30}
   }
   n13 = "Transform (Modify) Bone - Bone: base_skirt_back_lf"!(ComponentPose=@n12, Alpha:float=1, Translation="0,0,0", Rotation="25,0,0", Scale="1,1,1") -> (Pose) {
    rotationMode=BMM_Additive
    rotationSpace=BCS_BoneSpace
    translationMode=BMM_Additive
    translationSpace=BCS_BoneSpace
    alphaBoolBlend {blendInTime=0.5, blendOption=Sinusoidal, blendOutTime=1}
    boneToModify {boneName=base_skirt_back_lf}
    rotation {pitch=25}
   }
   n14 = "Transform (Modify) Bone - Bone: base_skirt_back"!(ComponentPose=@n13, Alpha:float=1, Translation="0,0,0", Rotation="40,0,0", Scale="1,1,1") -> (Pose) {
    rotationMode=BMM_Additive
    rotationSpace=BCS_BoneSpace
    translationMode=BMM_Additive
    translationSpace=BCS_BoneSpace
    alphaBoolBlend {blendInTime=0.5, blendOption=Sinusoidal, blendOutTime=1}
    boneToModify {boneName=base_skirt_back}
    rotation {pitch=40}
   }
   n15 = "Transform (Modify) Bone - Bone: base_skirt_back_rt"!(ComponentPose=@n14, Alpha:float=1, Translation="0,0,0", Rotation="50,0,0", Scale="1,1,1") -> (Pose) {
    rotationMode=BMM_Additive
    rotationSpace=BCS_BoneSpace
    translationMode=BMM_Additive
    translationSpace=BCS_BoneSpace
    alphaBoolBlend {blendInTime=0.5, blendOption=Sinusoidal, blendOutTime=1}
    boneToModify {boneName=base_skirt_back_rt}
    rotation {pitch=50}
   }
   n16 = "Transform (Modify) Bone - Bone: base_skirt_midback_rt"(ComponentPose=@n15, Alpha=1, Translation="0,0,0", Rotation="50,0,0", Scale="1,1,1") {
    rotationMode=BMM_Additive
    rotationSpace=BCS_BoneSpace
    translationMode=BMM_Additive
    translationSpace=BCS_BoneSpace
    alphaBoolBlend {blendInTime=0.5, blendOption=Sinusoidal, blendOutTime=1}
    boneToModify {boneName=base_skirt_midback_rt}
    rotation {pitch=50}
   }
   n17 = "Transform (Modify) Bone - Bone: base_skirt_mid_rt"!(ComponentPose=@n16, Alpha:float=1, Translation="0,0,0", Rotation="35,0,0", Scale="1,1,1") -> (Pose) {
    rotationMode=BMM_Additive
    rotationSpace=BCS_BoneSpace
    translationMode=BMM_Additive
    translationSpace=BCS_BoneSpace
    alphaBoolBlend {blendInTime=0.5, blendOption=Sinusoidal, blendOutTime=1}
    boneToModify {boneName=base_skirt_mid_rt}
    rotation {pitch=35}
   }
   n18 = "Transform (Modify) Bone - Bone: base_bag_back_lf"(ComponentPose=@n17, Alpha=1, Translation="0,0,0", Rotation="20,0,30", Scale="1,1,1") {
    rotationMode=BMM_Additive
    rotationSpace=BCS_BoneSpace
    translationMode=BMM_Additive
    translationSpace=BCS_BoneSpace
    alphaBoolBlend {blendInTime=0.5, blendOption=Sinusoidal, blendOutTime=1}
    boneToModify {boneName=base_bag_back_lf}
    rotation {pitch=20, roll=30}
   }
   n19 = "Transform (Modify) Bone - Bone: base_bag_front_rt"(ComponentPose=@n18, Alpha=1, Translation="0,0,0", Rotation="10,0,-30", Scale="1,1,1") {
    rotationMode=BMM_Additive
    rotationSpace=BCS_BoneSpace
    translationMode=BMM_Additive
    translationSpace=BCS_BoneSpace
    alphaBoolBlend {blendInTime=0.5, blendOption=Sinusoidal, blendOutTime=1}
    boneToModify {boneName=base_bag_front_rt}
    rotation {pitch=10, roll=-30}
   }
   n20 = "Transform (Modify) Bone - Bone: base_bag_back_rt"(ComponentPose=@n19, Alpha=1, Translation="0,0,0", Rotation="55,0,-5", Scale="1,1,1") {
    rotationMode=BMM_Additive
    rotationSpace=BCS_BoneSpace
    translationMode=BMM_Additive
    translationSpace=BCS_BoneSpace
    alphaBoolBlend {blendInTime=0.5, blendOption=Sinusoidal, blendOutTime=1}
    boneToModify {boneName=base_bag_back_rt}
    rotation {pitch=55, roll=-5}
   }
   n21 = "Transform (Modify) Bone - Bone: base_skirt_back_lf"(ComponentPose=@n20, Alpha=1, Translation="0,0,0", Rotation="33,0,0", Scale="1,1,1") {
    rotationMode=BMM_Additive
    rotationSpace=BCS_BoneSpace
    translationMode=BMM_Additive
    translationSpace=BCS_BoneSpace
    alphaBoolBlend {blendInTime=0.5, blendOption=Sinusoidal, blendOutTime=1}
    boneToModify {boneName=base_skirt_back_lf}
    rotation {pitch=33}
   }
   n22 = "Transform (Modify) Bone - Bone: base_skirt_back"(ComponentPose=@n21, Alpha=1, Translation="0,0,0", Rotation="15,0,0", Scale="1,1,1") {
    rotationMode=BMM_Additive
    rotationSpace=BCS_BoneSpace
    translationMode=BMM_Additive
    translationSpace=BCS_BoneSpace
    alphaBoolBlend {blendInTime=0.5, blendOption=Sinusoidal, blendOutTime=1}
    boneToModify {boneName=base_skirt_back}
    rotation {pitch=15}
   }
   n23 = "Transform (Modify) Bone - Bone: base_skirt_front_rt"(ComponentPose=@n22, bAlphaBoolEnabled=True, Translation="-1,-1,0", Rotation="5,0,0", Scale="1,1,1") {
    alphaInputType=Bool
    rotationMode=BMM_Additive
    rotationSpace=BCS_BoneSpace
    translationMode=BMM_Additive
    translationSpace=BCS_BoneSpace
    alphaBoolBlend {blendInTime=0.5, blendOption=QuadraticInOut, blendOutTime=1}
    boneToModify {boneName=base_skirt_front_rt}
    rotation {pitch=5}
    translation=<-1 -1>
   }
   n24 = "Transform (Modify) Bone - Bone: base_skirt_front_lf"(ComponentPose=@n23, bAlphaBoolEnabled=True, Translation="-2,-2,0", Rotation="8,0,0", Scale="1,1,1") {
    alphaInputType=Bool
    rotationMode=BMM_Additive
    rotationSpace=BCS_BoneSpace
    translationMode=BMM_Additive
    translationSpace=BCS_BoneSpace
    alphaBoolBlend {blendInTime=0.5, blendOption=QuadraticInOut, blendOutTime=1}
    boneToModify {boneName=base_skirt_front_lf}
    rotation {pitch=8}
    translation=<-2 -2>
   }
   Output Pose(Result=Component To Local(ComponentPose=@n24)) {layerGroup=None}
  ]}
 EventGraph {nodes [
   Event BlueprintUpdateAnimation()
   > n1 = Sequence!(execute) -> (then_0, then_1, then_2:Exec, then_3:Exec)
   > n2 = Gravity Override State Function DEPRECATED(
    CharacterPartAnimInstance=Self-Reference()
    GameTimeInSeconds=GetGameTimeInSeconds()
    GravityOverrideParams=Get Gravity Override Params()
    RBAN Params=Get RBAN Params()
   )
   > Set GravityOverride(GravityOverride=@n2)
   > Set Sim Space Settings(Sim Space Settings:Sim Space Settings Structure=Break RBANSetting Struct!(RBANSettingStruct:"RBANSetting Struct Structure (by ref)"=@n2.RBAN SettingsOut) -> (
     ComponentLinearAccScale_28_#1:Vector=~
     ComponentLinearVelScale_31_#2:Vector=~
     SimSpaceSettings_34_#3:Sim Space Settings Structure
    ) {is=title K2Node_BreakStruct}.SimSpaceSettings_34_#3) -> (Output_Get:Sim Space Settings Structure)
   n3 = OR Boolean(A=NOT Boolean(A=Get bRigidBodyEnabled()), B="integer >= integer"(A=Get CurrentLOD(), B=1))
   Set ASWonOrLOD1(execute=@n1.then_2, ASWonOrLOD1:bool=@n3) -> (Output_Get:bool)
   n4 = Get bIsSliding()
   n5 = Get bIsCrouching()
   > Set ASWonOrLOD1AndSlidingOrCrouching(ASWonOrLOD1AndSlidingOrCrouching:bool=AND Boolean(A=@n3, B=OR Boolean(A=@n4, B=@n5))) -> (Output_Get:bool)
   n6 = Get bIsMoving2D()
   n7 = Get bIsCrouching()
   Set CrouchingNotMove(execute=@n1.then_1, CrouchingNotMove:bool=AND Boolean(A=@n7, B=NOT Boolean(A=@n6))) -> (Output_Get:bool)
   > Set StandUpMoving(StandUpMoving:bool=AND Boolean(A=NOT Boolean(A=@n7), B=@n6)) -> (Output_Get:bool)
   n8 = Branch(execute=@n1.then_3, Condition=Get bIsSliding())
   > n9 = Set Left_Calf(Left_Calf:double=GetAlphaFromReferenceTransformDelta(SkeletalMeshComponent=GetPartSkeletalMeshComponent(), BoneName=calf_l, BaseName=thigh_l, InputRange=(X=0,Y=1), TwistAxis="1,0,0").RawOutput) -> (Output_Get:double)
   > n10 = Branch(Condition=NOT Boolean(A="float > float"(A=@n9, B=100)))
   > Set Knee_Slide(Knee_Slide:double=0) -> (Output_Get:double)
   > Set Baseball_Slide(Baseball_Slide:double=1) -> (Output_Get:double)
   Set Knee_Slide(execute=@n10.else, Knee_Slide=1)
   > Set Baseball_Slide(Baseball_Slide=0)
   Set Knee_Slide(execute=@n8.else, Knee_Slide=0)
   > Set Baseball_Slide(Baseball_Slide=0)
  ]}
]
variables [
 ASWonOrLOD1 {type=boolean}
 ASWonOrLOD1AndSlidingOrCrouching {type=boolean}
 Baseball_Slide {type=number}
 CrouchingNotMove {type=boolean}
 Gravity Override Params {type=GravityOverrideParamsStruct}
 GravityOverride {type=Vector}
 Knee_Slide {type=number}
 Left_Calf {type=number}
 RBAN Params {
  type=RBANParamsStruct
  default {
   dBNO {simSpaceSettings {
     maxAngularAcceleration=1000
     maxAngularVelocity=3
     maxLinearAcceleration=300
     maxLinearVelocity=700
     velocityScaleZ=0
     externalLinearDragV=<0 0.5 0>
    }}
   dBNO_Carried {simSpaceSettings {
     maxAngularAcceleration=1000
     maxAngularVelocity=3
     maxLinearAcceleration=300
     maxLinearVelocity=700
     velocityScaleZ=0
     externalLinearDragV=<0 0.5 0>
    }}
   emoteOrMelee {simSpaceSettings {
     maxAngularAcceleration=1000
     maxAngularVelocity=3
     maxLinearAcceleration=300
     maxLinearVelocity=700
     velocityScaleZ=0
     externalLinearDragV=<0 0.5 0>
    }}
   front_End {simSpaceSettings {
     maxAngularAcceleration=1000
     maxAngularVelocity=3
     maxLinearAcceleration=300
     maxLinearVelocity=700
     velocityScaleZ=0
     externalLinearDragV=<0 0.5 0>
    }}
   onGround_Crouch_Moving {simSpaceSettings {
     maxAngularAcceleration=1000
     maxAngularVelocity=3
     maxLinearAcceleration=300
     maxLinearVelocity=700
     velocityScaleZ=0
     externalLinearDragV=<0 0.5 0>
    }}
   onGround_Falling {simSpaceSettings {
     maxAngularAcceleration=1000
     maxAngularVelocity=3
     maxLinearAcceleration=300
     maxLinearVelocity=700
     velocityScaleZ=0
     externalLinearDragV=<0 0.5 0>
    }}
   onGround_Moving {simSpaceSettings {
     maxAngularAcceleration=1000
     maxAngularVelocity=3
     maxLinearAcceleration=300
     maxLinearVelocity=700
     velocityScaleZ=0
     externalLinearDragV=<0 0.5 0>
    }}
   onGround_Standing {simSpaceSettings {
     maxAngularAcceleration=1000
     maxAngularVelocity=3
     maxLinearAcceleration=300
     maxLinearVelocity=700
     velocityScaleZ=0
     externalLinearDragV=<0 0.5 0>
    }}
   rail_Grinding {simSpaceSettings {
     maxAngularAcceleration=1000
     maxAngularVelocity=3
     maxLinearAcceleration=300
     maxLinearVelocity=700
     velocityScaleZ=0
     worldAlpha=0.1
     externalLinearDragV=<0 0.5 0>
    }}
   riding_Moving {simSpaceSettings {
     maxAngularAcceleration=1000
     maxAngularVelocity=3
     maxLinearAcceleration=300
     maxLinearVelocity=700
     velocityScaleZ=0
     externalLinearDragV=<0 0.5 0>
    }}
   riding_Moving_Targeting {simSpaceSettings {
     maxAngularAcceleration=1000
     maxAngularVelocity=3
     maxLinearAcceleration=300
     maxLinearVelocity=700
     velocityScaleZ=0
     externalLinearDragV=<0 0.5 0>
    }}
   riding_Standing {simSpaceSettings {
     maxAngularAcceleration=1000
     maxAngularVelocity=3
     maxLinearAcceleration=300
     maxLinearVelocity=700
     velocityScaleZ=0
     externalLinearDragV=<0 0.5 0>
    }}
   skydive_DivingDown {simSpaceSettings {
     maxAngularAcceleration=1000
     maxAngularVelocity=3
     maxLinearAcceleration=300
     maxLinearVelocity=700
     velocityScaleZ=0
     externalLinearDragV=<0 0.5 0>
    }}
   skydive_FloatingDown {
    componentLinearAccScale=<0 0.5 0>
    componentLinearVelScale=<0 0.5 0>
    simSpaceSettings {
     maxAngularAcceleration=1000
     maxAngularVelocity=3
     maxLinearAcceleration=300
     maxLinearVelocity=700
     velocityScaleZ=0
     externalLinearDragV=<0 0.5 0>
    }
   }
   skydive_GoingUp {simSpaceSettings {
     maxAngularAcceleration=1000
     maxAngularVelocity=3
     maxLinearAcceleration=300
     maxLinearVelocity=700
     velocityScaleZ=0
     externalLinearDragV=<0 0.5 0>
    }}
   skydive_Parachuting {
    componentLinearAccScale=<0 0.1 0.3>
    componentLinearVelScale=<0 0.1 0.3>
    simSpaceSettings {
     maxAngularAcceleration=1000
     maxAngularVelocity=3
     maxLinearAcceleration=300
     maxLinearVelocity=700
     velocityScaleZ=0
     externalLinearDragV=<0 0.5 0>
    }
   }
   skydive_Surfing {simSpaceSettings {
     maxAngularAcceleration=1000
     maxAngularVelocity=3
     maxLinearAcceleration=300
     maxLinearVelocity=700
     velocityScaleZ=0
     externalLinearDragV=<0 0.5 0>
    }}
   swimming {simSpaceSettings {
     maxAngularAcceleration=1000
     maxAngularVelocity=3
     maxLinearAcceleration=300
     maxLinearVelocity=700
     velocityScaleZ=0
     externalLinearDragV=<0 0.5 0>
    }}
   tactical_Sprint {simSpaceSettings {
     maxAngularAcceleration=1000
     maxAngularVelocity=3
     maxLinearAcceleration=300
     maxLinearVelocity=700
     velocityScaleZ=0
     externalLinearDragV=<0 0.5 0>
    }}
   ziplining {simSpaceSettings {
     maxAngularAcceleration=1000
     maxAngularVelocity=3
     maxLinearAcceleration=300
     maxLinearVelocity=700
     velocityScaleZ=0
     externalLinearDragV=<0 0.5 0>
    }}
  }
 }
 Sim Space Settings {
  type=SimSpaceSettings
  default {
   dampingAlpha=1
   maxAngularAcceleration=10000
   maxAngularVelocity=10000
   maxLinearAcceleration=10000
   maxLinearVelocity=10000
   velocityScaleZ=1
   worldAlpha=0
   externalLinearDragV=<0 0 0>
  }
 }
 StandUpMoving {type=boolean}
]
