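# Animation Blueprint for the Feudal Lord face accessory part of a Fortnite player character, following the parent mesh's pose.
# It layers the Body mesh's clavicle and upper-arm branches over that copied pose.
# An additive head rotation counters the head's ref-pose delta from neck_02 by per-axis percentages, then feeds a Control Rig that takes crouch state.
# A deprecated helper also updates GravityOverride, which the AnimGraph does not read directly.
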AnimBlueprint /BRCosmetics/Characters/Player/Male/Medium/Bodies/M_MED_FeudalLord/Meshes/Parts/M_MED_FeudalLord_FaceAcc_AnimBP
target_skeleton=&/M_MED_FeudalLord_FaceAcc_Skeleton
graphs [
 AnimGraph {nodes [
   n1 = "Transform (Modify) Bone - Bone: head"(
    ComponentPose=Local To Component(LocalPose=Use cached pose 'StartPose'())
    Alpha=1
    Translation="0,0,0"
    Rotation=Get head rot() -> (head rot:Rotator)
    Scale="1,1,1"
   ) {
    rotationMode=BMM_Additive
    rotationSpace=BCS_BoneSpace
    alphaBoolBlend {blendOption=Linear}
    boneToModify {boneName=head}
    rotation=<rot 50 50 50>
   }
   n2 = Control Rig!(Source=Component To Local(ComponentPose=@n1), Alpha, Crouching:bool=Get bIsCrouching()) -> (Pose) {
    alphaBoolBlend {blendOption=Linear}
    controlRigAssetReference {blueprintRigClass=&/M_MED_FeudalLord_FaceAcc_CtrlRig.M_MED_FeudalLord_FaceAcc_CtrlRig_C}
   }
   Output Pose(Result=@n2) {layerGroup=None}
   n3 = Layered blend per bone(
    BasePose=Copy Pose From Mesh() {bCopyCurves=true, bUseAttachedParent=true}
    BlendPoses_0=Copy Pose From Mesh(SourceMeshComponent=Get Body_Mesh())
   ) {layerSetup [{branchFilters [{boneName=clavicle_l}, {boneName=clavicle_r}, {boneName=upperarm_r}, {boneName=upperarm_l}]}]}
   Save cached pose 'StartPose'(Pose=@n3) {cacheName=StartPose}
  ]}
 EventGraph {nodes [
   Event OnCharacterCustomizationLoaded()
   > n1 = Cast To PlayerPawn_Athena(Object=TryGetPawnOwner())
   > Set Body_Mesh(Body_Mesh=GetSkeletalMeshForPartType(self=@n1))
   Event BlueprintUpdateAnimation()
   > n2 = Gravity Override State Function DEPRECATED(CharacterPartAnimInstance=Self-Reference(), GameTimeInSeconds=GetGameTimeInSeconds(), GravityOverrideParams=Get Gravity Override Params())
   > Set GravityOverride(GravityOverride=@n2)
   n3 = GetOwnerSkeletalMeshComponent()
   n4 = BreakRotator(InRot=BreakTransform(InTransform=GetDeltaTransformFromRefPose(self=@n3, BoneName=head, BaseName=neck_02)).Rotation)
   n5 = Get Y percentage() -> (Y percentage:float)
   n6 = Get X percentage() -> (X percentage:float)
   n7 = Get Z percentage() -> (Z percentage:float)
   > Set head rot(head rot:Rotator=MakeRotator(
     Roll=float * float!(A=@n4, B=-1, C:double=@n6) -> (ReturnValue)
     Pitch=float * float(A=@n4.Pitch, B=-1, C=@n5)
     Yaw=float * float(A=@n4.Yaw, B=-1, C=@n7)
    )) -> (Output_Get:Rotator)
  ]}
]
variables [
 Body_Mesh {type=SkeletalMeshComponent}
 Gravity Override Params {type=GravityOverrideParamsStruct}
 GravityOverride {type=Vector}
 head rot {type=Rotator}
 X percentage {type=number}
 Y percentage {type=number}
 Z percentage {type=number}
]
